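; Blueprint function library that provides access to a test message box widget and a way to post test messages to it.
; getTestMessageBox returns the first existing widget of a soft-referenced class, or creates one and adds it to the viewport at a very high Z-order.
; addTestMessage sends a message struct to that widget through the BPTestWidgetInterface and can optionally force the widget to show.
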(fn addTestMessage
  (params (message :type Test%20Message%20Struct%20Structure) (forceShowMenu :type Boolean))
  (bind _box (CallFunction|GetTestMessageBox 0))
  (Class|BPTestWidgetInterface|AddMsg(Message) _box message)
  (if forceShowMenu
    (Class|BPTestWidgetInterface|ShowWidget(Message) _box)))

(fn getTestMessageBox
  (bind _foundwidgets (Widget|GetAllWidgetsOfClass 0 (Utilities|ResolveSoftReference (LocalVariables|Getl_widgetSoftRef)) false))
  (if (Utilities|Array|IsValidIndex _foundwidgets)
    (return (Utilities|Array|Get(acopy) _foundwidgets))
    (else
      (bind _returnvalue (UserInterface|CreateWidget 0 (Utilities|ResolveSoftReference (LocalVariables|Getl_widgetSoftRef))))
      (UserInterface|Viewport|AddtoViewport _returnvalue 999999)
      (return _returnvalue))))
