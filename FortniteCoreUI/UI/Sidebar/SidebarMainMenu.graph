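# SidebarMainMenu is the Fortnite sidebar main menu widget, with buttons for credits, legal, support, respawn, team select, minigame controls, creative moderator mode and dev tools.
# On construct and refresh, helper functions set button visibility from runtime options, platform, game state and permissions.
# SwitchSubMenu drives a widget switcher for Reporting, Support and Legal submenus, tracking the last selected button for focus.
# Many buttons async-create soft-class screens and push them as panels.
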WidgetBlueprint /FortniteCoreUI/UI/Sidebar/SidebarMainMenu
animations [Drill_In {
  binding_count=2
  length_seconds=1
  object=&self:Drill_In
  sequence {bindings [
    Switcher_Menus {
     id=#1
     display_name=Switcher_SecondaryMenu
     possessed_class=VerticalBox
     tracks [
      {
       type=MovieSceneFloatTrack
       display_name=Render Opacity
       sections [{type=MovieSceneFloatSection, end_frame=20, start_frame=0, channels [None {type=MovieSceneScriptingFloatChannel, keys [[0,0],[5,0],[10,1]]}]}]
      }
      {
       type=MovieScene2DTransformTrack
       sections [{
         type=MovieScene2DTransformSection
         end_frame=20
         start_frame=0
         channels [
          Translation.X {type=MovieSceneScriptingFloatChannel, keys [[0,500],[20,0]]}
          Translation.Y {type=MovieSceneScriptingFloatChannel, const [0,1]}
         ]
        }]
      }
     ]
    }
    VBox_MainMenuButtons {
     id=#2
     possessed_class=VerticalBox
     tracks [
      {
       type=MovieScene2DTransformTrack
       sections [{
         type=MovieScene2DTransformSection
         end_frame=20
         start_frame=0
         channels [
          Translation.X {type=MovieSceneScriptingFloatChannel, keys [[0,0],[20,-500]]}
          Translation.Y {type=MovieSceneScriptingFloatChannel, const [0,1]}
         ]
        }]
      }
      {
       type=MovieSceneFloatTrack
       display_name=Render Opacity
       sections [{type=MovieSceneFloatSection, end_frame=20, start_frame=0, channels [None {type=MovieSceneScriptingFloatChannel, keys [[0,1],[5,0]]}]}]
      }
     ]
    }
   ]}
 }]
graphs [
 BP_GetDesiredFocusTarget {nodes [
   Get Desired Focus Target()
   > n1 = Cast To VerticalBox!(
    execute
    Object=GetActiveWidget!(self:Widget Switcher Object Reference=Get Switcher_Menus() -> (Switcher_Menus:Widget Switcher Object Reference)) -> (ReturnValue:Widget Object Reference)
   ) -> (then, CastFailed, AsVertical Box:Vertical Box Object Reference) {is=title K2Node_DynamicCast}
   n2 = Get LastSelectedButton() -> (LastSelectedButton:Common Button Legacy Object Reference)
   > n3 = Branch(Condition=HasChild!(self:Panel Widget Object Reference=@n1, Content:Widget Object Reference=@n2) -> (ReturnValue:bool))
   > Return Node!(execute, ReturnValue:Widget Object Reference=@n2)
   n4 = Get ConsoleCommandPrompt() -> (ConsoleCommandPrompt:WBP UIKit Input Field Object Reference)
   n5 = Get EditableText!(self:WBP UIKit Block Editable Text Object Reference=Get Block_EditableText!(self:WBP UIKit Input Field Base Object Reference=@n4) -> (Block_EditableText:WBP UIKit Block Editable Text Object Reference)) -> (EditableText:UIKit Editable Text Object Reference)
   Return Node(
    execute=[@n3.else, @n1.CastFailed]
    ReturnValue=Select!(
     Visible:Widget Object Reference=@n5
     Collapsed:Widget Object Reference
     Hidden:Widget Object Reference
     HitTestInvisible:Widget Object Reference=@n5
     SelfHitTestInvisible:Widget Object Reference=@n5
     Index:ESlateVisibility Enum=GetVisibility!(self:Widget Object Reference=@n4) -> (ReturnValue:ESlateVisibility Enum=Visible)
    ) -> (ReturnValue:Widget Object Reference)
   )
  ]}
 Close {nodes [
   Close() {is=title K2Node_FunctionEntry}
   > n1 = FindParentWidgetOfType!(execute, StartingWidget:Widget Object Reference=Self-Reference(), Type:Widget Class Reference=/Script/FortniteUI.FortDualBladeMenu) -> (then, ReturnValue:Fort Dual Blade Menu Object Reference)
   > n2 = Cast To DualBladeMenu!(execute, Object=@n1) -> (then, CastFailed, AsDual Blade Menu:Dual Blade Menu Object Reference) {is=title K2Node_DynamicCast}
   > DeactivateWidget(self=@n2)
  ]}
 EventGraph {nodes [
   "On Button Clicked (ParentalControlsButton)"!() -> (OutputDelegate, then, Button:Common Button Legacy Object Reference) {is=title K2Node_ComponentBoundEvent}
   > RequestToOpenParentalControls!(execute) -> (then)
   "On Button Clicked (RespawnButton)"!() -> (OutputDelegate, then, Button:Common Button Legacy Object Reference) {is=title K2Node_ComponentBoundEvent}
   > n1 = CreateWidgetAsync!(
    execute:Exec
    UserWidgetSoftClass:User Widget Soft Class Reference=Get RespawnClass() -> (RespawnClass:User Widget Soft Class Reference)
    OwningPlayer:Player Controller Object Reference=GetOwningPlayer()
    bSuspendInputUntilComplete:bool=true
   ) -> (then:Exec, OnComplete:Exec, OnFailure:Exec, UserWidget:User Widget Object Reference) {is=title K2Node_AsyncAction}
   TryPushWidgetAsModalPanel!(execute=@n1.OnComplete, self:Fort Blade Menu Main Menu Object Reference, InWidget:User Widget Object Reference=@n1) -> (then)
   "On Button Clicked (CheatsButton)"!() -> (OutputDelegate, then, Button:Common Button Legacy Object Reference) {is=title K2Node_ComponentBoundEvent}
   > n2 = CreateWidgetAsync(
    UserWidgetSoftClass=/FortniteCoreUI/Settings/SettingsScreen_GameCheats.SettingsScreen_GameCheats_C
    OwningPlayer=GetOwningPlayer()
    bSuspendInputUntilComplete=true
   )
   TryPushWidgetAsActivatablePanel!(
    execute=@n2.OnComplete
    self:Fort Blade Menu Main Menu Object Reference
    InWidget:User Widget Object Reference=@n2
    bHideHeader:bool=true
    bHideFooter:bool=false
    bHideChatWidget:bool=true
   ) -> (then)
   "On Button Clicked (Button_MyIsland)"!() -> (OutputDelegate, then, Button:Common Button Legacy Object Reference) {is=title K2Node_ComponentBoundEvent}
   > n3 = Cast To FortPlayerController(Object=GetOwningPlayer())
   > n4 = Branch(Condition=ShouldAddTabsToMapScreen!(PlayerController:Player Controller Object Reference=@n3) -> (ReturnValue:bool))
   n5 = GetMiniMapComponent!(self:Fort Player Controller Object Reference=@n3) -> (ReturnValue:Fort Controller Component Minimap Object Reference)
   > Is Valid(InputObject=@n5)
   > OpenFullScreenMapTab!(execute, self:Fort Controller Component Minimap Object Reference=@n5, TabId:Name=GetMyIslandTabID!() -> (ReturnValue:Name=None)) -> (then)
   OpenInventory!(execute=@n4.else, self:Fort Player Controller Object Reference=@n3, InventoryTabId:Name=GetMyIslandTabID() -> (ReturnValue=None)) -> (then)
   "On Button Clicked (ButtonCredits)"!() -> (OutputDelegate, then, Button:Common Button Legacy Object Reference) {is=title K2Node_ComponentBoundEvent}
   > n6 = CreateWidgetAsync(
    UserWidgetSoftClass=Get CreditsScreenClass() -> (CreditsScreenClass:User Widget Soft Class Reference)
    OwningPlayer=GetOwningPlayer()
    bSuspendInputUntilComplete=true
   )
   n7 = Branch(execute=@n6.OnComplete, Condition=IsMobileGame())
   n8 = Get FortFrontEndContext!() -> (ReturnValue:Fort Front End Context Object Reference) {is=title K2Node_GetBlueprintContextEx}
   > PushContentWidget_Adv!(
    execute
    self:Fort Front End Context Object Reference=@n8
    Widget:Widget Object Reference=@n6
    bHideHeader:bool=true
    bHideFooter:bool=true
    bHideChatWidget:bool=true
   ) -> (then)
   PushContentWidget_Adv(execute=@n7.else, self=@n8, Widget=@n6, bHideHeader=true, bHideFooter=false, bHideChatWidget=true)
   n9 = "On Button Clicked (NewPlayerFeedback)"!() -> (OutputDelegate, then, Button:Common Button Legacy Object Reference) {is=title K2Node_ComponentBoundEvent}
   > SwitchSubMenu!(execute, self, Index:int=1) -> (then)
   > Set LastSelectedButton(LastSelectedButton:Common Button Legacy Object Reference=@n9.Button) -> (Output_Get:Common Button Legacy Object Reference)
   "On Clicked (ConsoleCommandPrompt)"!() -> (OutputDelegate, then, Button:Common Button Base Object Reference) {is=title K2Node_ComponentBoundEvent}
   > Set LastSelectedButton()
   "On Button Clicked (EnableModeratorModeButton)"!() -> (OutputDelegate, then, Button:Common Button Legacy Object Reference) {is=title K2Node_ComponentBoundEvent}
   > Branch(Condition=Get bAllowCreativeModeratorMode!(self:Fort Runtime Options Object Reference=GetRuntimeOptions()) -> (bAllowCreativeModeratorMode:bool))
   > n10 = Cast To FortPlayerControllerGameplay!(execute, Object=GetOwningPlayer()) -> (then, CastFailed, AsFort Player Controller Gameplay:Fort Player Controller Gameplay Object Reference) {is=title K2Node_DynamicCast}
   > SetVisibility(self=Get EnableModeratorModeButton() -> (EnableModeratorModeButton:Main Menu Button Object Reference), InVisibility=Collapsed)
   > "SetText (Text)"(
    self=Get ModeratorModeLoadingMessage() -> (ModeratorModeLoadingMessage:Common Text Object Reference)
    InText="LOCTEXT(\"Enabling Moderator Mode\")"
   )
   > SetVisibility(self=Get ModerationModeLoading() -> (ModerationModeLoading:Horizontal Box Object Reference), InVisibility=SelfHitTestInvisible)
   > SetCreativeModeratorModeActive!(execute, self:Fort Player Controller Gameplay Object Reference=@n10, bActive:bool=true) -> (then)
   "On Button Clicked (DisableModeratorModeButton)"!() -> (OutputDelegate, then, Button:Common Button Legacy Object Reference) {is=title K2Node_ComponentBoundEvent}
   > Branch(Condition=Get bAllowCreativeModeratorMode(self=GetRuntimeOptions()))
   > n11 = Cast To FortPlayerControllerGameplay(Object=GetOwningPlayer())
   > SetVisibility(self=Get DisableModeratorModeButton() -> (DisableModeratorModeButton:Main Menu Button Object Reference), InVisibility=Collapsed)
   > "SetText (Text)"(self=Get ModeratorModeLoadingMessage(), InText="LOCTEXT(\"Disabling Moderator Mode\")")
   > SetVisibility(self=Get ModerationModeLoading(), InVisibility=SelfHitTestInvisible)
   > SetCreativeModeratorModeActive(self=@n11, bActive=false)
   "On Button Clicked (Button_ReportBugOrPlayer)"!() -> (OutputDelegate, then, Button:Common Button Legacy Object Reference) {is=title K2Node_ComponentBoundEvent}
   > RequestToOpenPlayerFeedbackModal!(execute, Context:User Widget Object Reference=Self-Reference(), UserToReport:Social User Object Reference) -> (then)
   > ReleaseSavedMenuPosition(self:Fort Blade Menu Main Menu Object Reference)
   n12 = "On Selected Changed (ConsoleCommandPrompt)"!() -> (OutputDelegate, then, Button:Common Button Base Object Reference, Selected:bool) {is=title K2Node_ComponentBoundEvent}
   > Branch(Condition=@n12.Selected)
   > Set LastSelectedButton()
   "On Button Clicked (ReturnOrCancelPurchase)"!() -> (OutputDelegate, then, Button:Common Button Legacy Object Reference) {is=title K2Node_ComponentBoundEvent}
   > ViewPurchaseHistory(self:Fort Blade Menu Main Menu Object Reference)
   n13 = "On Clicked (SidebarBackButton)"!() -> (OutputDelegate, then, Button:Common Button Base Object Reference) {is=title K2Node_ComponentBoundEvent}
   SwitchSubMenu(execute=[@n13.then, @n46.then], Index=0)
   "On Button Clicked (ButtonHUDLayoutTool)"!() -> (OutputDelegate, then, Button:Common Button Legacy Object Reference) {is=title K2Node_ComponentBoundEvent}
   n14 = Get FortUIDataConfigurationContext!() -> (ReturnValue:Fort UIData Configuration Context Object Reference) {is=title K2Node_GetBlueprintContextEx}
   n15 = GetSoftWidgetClassForAlias!(
    self:Fort UIData Configuration Context Object Reference=@n14
    WidgetAlias:"Gameplay Tag Structure (by ref)"=MakeLiteralGameplayTag!(Value:Gameplay Tag Structure=(TagName="MobileHUD.HUDLayoutTool")) -> (ReturnValue:Gameplay Tag Structure)
   ) -> (ReturnValue:User Widget Soft Class Reference)
   n16 = Select!(
    Option 0:User Widget Soft Class Reference=GetHUDLayoutToolV2SoftClass!(self:Fort UIData Configuration Context Object Reference=@n14) -> (ReturnValue:Fort Mobile HUDLayout Tool V2 Soft Class Reference)
    Option 1:User Widget Soft Class Reference=@n15
    Index:bool=IsValidSoftClassReference!(SoftClassReference:"Object Soft Class Reference (by ref)"=@n15) -> (ReturnValue:bool)
   ) -> (ReturnValue:User Widget Soft Class Reference)
   > Branch(Condition=IsValidSoftClassReference(SoftClassReference=@n16))
   > n17 = CreateWidgetAsync(UserWidgetSoftClass=@n16, OwningPlayer=GetOwningPlayer(), bSuspendInputUntilComplete=true)
   TryPushWidgetAsActivatablePanel(execute=@n17.OnComplete, InWidget=@n17, bHideHeader=false, bHideFooter=false, bHideChatWidget=true)
   "On Button Clicked (Button_LegalInformation)"!() -> (OutputDelegate, then, Button:Common Button Legacy Object Reference) {is=title K2Node_ComponentBoundEvent}
   > n18 = CreateWidgetAsync(
    UserWidgetSoftClass=Get LegalInfoClass() -> (LegalInfoClass:User Widget Soft Class Reference)
    OwningPlayer=GetOwningPlayer()
    bSuspendInputUntilComplete=true
   )
   TryPushWidgetAsActivatablePanel(execute=@n18.OnComplete, InWidget=@n18, bHideHeader=true, bHideFooter=true, bHideChatWidget=true)
   n19 = "On Button Clicked (Button_SupportSubMenu)"!() -> (OutputDelegate, then, Button:Common Button Legacy Object Reference) {is=title K2Node_ComponentBoundEvent}
   > SwitchSubMenu(Index=2)
   > Set LastSelectedButton(LastSelectedButton=@n19.Button)
   "On Button Clicked (TeamSelectButton)"!() -> (OutputDelegate, then, Button:Common Button Legacy Object Reference) {is=title K2Node_ComponentBoundEvent}
   > n20 = CreateWidgetAsync(
    UserWidgetSoftClass=Get TeamSelectWidgetClass() -> (TeamSelectWidgetClass:User Widget Soft Class Reference)
    OwningPlayer=GetOwningPlayer()
    bSuspendInputUntilComplete=true
   )
   TryPushWidgetAsModalPanel(execute=@n20.OnComplete, InWidget=@n20)
   "On Button Clicked (ButtonBacchusHelp)"!() -> (OutputDelegate, then, Button:Common Button Legacy Object Reference) {is=title K2Node_ComponentBoundEvent}
   n21 = GetSoftWidgetClassForAlias(self=Get FortUIDataConfigurationContext(), WidgetAlias=MakeLiteralGameplayTag(Value=(TagName="MobileHUD.TouchControlsHelp")))
   > n22 = CreateWidgetAsync(
    UserWidgetSoftClass=Select(
     Option 0=Get MobileControlsHelpClass() -> (MobileControlsHelpClass:User Widget Soft Class Reference)
     Option 1=@n21
     Index=IsValidSoftClassReference(SoftClassReference=@n21)
    )
    OwningPlayer=GetOwningPlayer()
    bSuspendInputUntilComplete=true
   )
   TryPushWidgetAsActivatablePanel(execute=@n22.OnComplete, InWidget=@n22, bHideHeader=true, bHideFooter=true, bHideChatWidget=true)
   "Start Button Clicked (MinigameButtons)"() {is=title K2Node_ComponentBoundEvent}
   n23 = Get Machine() -> (Machine:Fort Minigame Settings Building Object Reference)
   > Is Valid(InputObject=@n23)
   > StartMinigame_BP(self:Fort Minigame Settings Building Object Reference=@n23)
   > Close() {is=title K2Node_CallFunction}
   "End Button Clicked (MinigameButtons)"() {is=title K2Node_ComponentBoundEvent}
   n24 = Get Machine()
   > Is Valid(InputObject=@n24)
   > AbandonMinigame_BP(self:Fort Minigame Settings Building Object Reference=@n24)
   > Close()
   n25 = "On Button Clicked (Legal)"!() -> (OutputDelegate, then, Button:Common Button Legacy Object Reference) {is=title K2Node_ComponentBoundEvent}
   > SwitchSubMenu(Index=3)
   > Set LastSelectedButton(LastSelectedButton=@n25.Button)
   "On Button Clicked (ExportIslandButton)"!() -> (OutputDelegate, then, Button:Common Button Legacy Object Reference) {is=title K2Node_ComponentBoundEvent}
   > Branch(Condition=Get bAllowIslandExporting!(self:Fort Runtime Options Object Reference=GetRuntimeOptions()) -> (bAllowIslandExporting:bool))
   > SpawnActor Creative Island Json Exporter!(
    execute:Exec
    Class:Actor Class Reference=/Script/FortniteGame.CreativeIslandJsonExporter
    SpawnTransform_Location:Vector=GetActorEyesViewPoint!(self:Actor Object Reference=GetOwningPlayer()) -> (OutLocation:Vector, OutRotation:Rotator)
    SpawnTransform_Rotation:Rotator="0, 0, 0"
    SpawnTransform_Scale:Vector="1,1,1"
    CollisionHandlingOverride:ESpawnActorCollisionHandlingMethod Enum=AlwaysSpawn
    TransformScaleMethod:ESpawnActorScaleMethod Enum=MultiplyWithRoot
    Owner:Actor Object Reference
    Instigator:Pawn Object Reference
   ) -> (then:Exec, ReturnValue:Creative Island Json Exporter Object Reference) {is=title K2Node_SpawnActorFromClass}
   "On Button Clicked (AthenaMainMenuClearPlayerDataButton)"!() -> (OutputDelegate, then, Button:Common Button Legacy Object Reference) {is=title K2Node_ComponentBoundEvent}
   > n26 = CreateWidgetAsync(
    UserWidgetSoftClass=Get PlayerClearIslandDataClass() -> (PlayerClearIslandDataClass:User Widget Soft Class Reference)
    OwningPlayer=GetOwningPlayer()
    bSuspendInputUntilComplete=true
   )
   n27 = Cast To PlayerClearPersistenceIslandDataPromptWidget!(execute=@n26.OnComplete, Object=@n26) -> (then, CastFailed, AsPlayer Clear Persistence Island Data Prompt Widget:Player Clear Persistence Island Data Prompt Widget Object Reference) {is=title K2Node_DynamicCast}
   n28 = Get Machine()
   > Is Valid(InputObject=@n28)
   n29 = GetMinigame!(self:Fort Minigame Settings Building Object Reference=@n28) -> (ReturnValue:Fort Minigame Object Reference)
   > SetDescription!(
    execute
    self:Player Clear Persistence Island Data Prompt Widget Object Reference=@n27
    InDescription:Text=Select(
     Option 0:Text=Get Player Clear Persistence Data Pregame Description() -> (Player Clear Persistence Data Pregame Description:Text)
     Option 1:Text=Get Player Clear Persistence Data in Progress Description() -> (Player Clear Persistence Data in Progress Description:Text)
     Index=IsInProgress!(self:Fort Minigame Object Reference=@n29) -> (ReturnValue:bool)
    ) -> (ReturnValue:Text)
   ) -> (then)
   > TryPushWidgetAsModalPanel(InWidget=@n26)
   "On Button Clicked (ConsoleButton)"!() -> (OutputDelegate, then, Button:Common Button Legacy Object Reference) {is=title K2Node_ComponentBoundEvent}
   > ShowConsoleWindow(self:Fort Blade Menu Main Menu Object Reference)
   n30 = "On Text Commited (ConsoleCommandPrompt)"!() -> (OutputDelegate, then, Text:Text, Commit method:ETextCommit Enum=Default) {is=title K2Node_ComponentBoundEvent}
   > Branch(Condition="Equal (Enum)"(A:ETextCommit Enum=@n30.Commit method, B:ETextCommit Enum=OnEnter))
   > ExecuteConsoleCommand!(execute, self:Fort Blade Menu Main Menu Object Reference, Command="To String (Text)"(InText=@n30.Text)) -> (then)
   UpdateMinigameButtons() {is=title K2Node_CustomEvent}
   > n31 = Cast To MinigameButtons!(execute, Object=Get MinigameButtons() -> (MinigameButtons:Fort Minigame Buttons Object Reference)) -> (then, CastFailed, AsMinigame Buttons:Minigame Buttons Object Reference) {is=title K2Node_DynamicCast}
   > UpdateButtons(self:Minigame Buttons Object Reference=@n31)
   Bind To State Change Delegate() {is=title K2Node_CustomEvent}
   n32 = Get Machine()
   > n33 = Is Valid(InputObject=@n32)
   n34 = OnMinigameStateChanged!() -> (OutputDelegate, then, Minigame:Fort Minigame Object Reference, MinigameState:EFortMinigameState Enum=PreGame) {is=title K2Node_CustomEvent}
   Bind Event to State Changed Delegate(execute=@n33.Is Valid, self:Fort Minigame Object Reference=GetMinigame(self=@n32), Delegate=@n34) {is=title K2Node_AddDelegate}
   On Possessed Pawn Changed!() -> (OutputDelegate, then, Old Pawn:Pawn Object Reference, New Pawn:Pawn Object Reference) {is=title K2Node_CustomEvent}
   > UpdateRespawnWidgetVisibility()
   n35 = Handle On Is Creative Moderator Mode Active Changed!() -> (OutputDelegate, then, bIsCreativeModeratorModeActive:bool=~) {is=title K2Node_CustomEvent}
   > n36 = Sequence!(execute) -> (then_0, then_1, then_2:Exec, then_3:Exec)
   > SetVisibility(self=Get ModerationModeLoading(), InVisibility=Collapsed)
   SetVisibility(
    execute=@n36.then_1
    self=Get EnableModeratorModeButton()
    InVisibility=Select(Option 0:ESlateVisibility Enum=Visible, Option 1:ESlateVisibility Enum=Collapsed, Index=@n35.bIsCreativeModeratorModeActive) -> (ReturnValue:ESlateVisibility Enum)
   )
   SetVisibility(
    execute=@n36.then_2
    self=Get DisableModeratorModeButton()
    InVisibility=Select(Option 0=Collapsed, Option 1=Visible, Index=@n35.bIsCreativeModeratorModeActive)
   )
   Update Team Select Visibility(execute=@n36.then_3)
   Event Construct()
   > InitializeMainMenu()
   > SetVisibility(self=Get NewPlayerFeedback() -> (NewPlayerFeedback:Main Menu Button Object Reference), InVisibility=Visible)
   n37 = Get FortGlobalUIContext!() -> (ReturnValue:Fort Global UIContext Object Reference) {is=title K2Node_GetBlueprintContextEx}
   n38 = GetRuntimeOptions()
   n39 = OR Boolean(
    A=Get bEnableParentalControlsInMatch!(self:Fort Runtime Options Object Reference=@n38) -> (bEnableParentalControlsInMatch:bool)
    B=NOT Boolean(A=IsInZone!(self:Fort Global UIContext Object Reference=@n37) -> (ReturnValue:bool))
   )
   n40 = Get bEnableContentControls!(self:Fort Runtime Options Object Reference=@n38) -> (bEnableContentControls:bool)
   > SetVisibility(
    self=Get ParentalControlsButton() -> (ParentalControlsButton:Main Menu Button Object Reference)
    InVisibility=Select(Option 0=Collapsed, Option 1=Visible, Index=AND Boolean(A=@n40, B=@n39))
   )
   n41 = GetRuntimeOptions()
   > SetVisibility(
    self=Get CheatsButton() -> (CheatsButton:Main Menu Button Object Reference)
    InVisibility=Select(Option 0=Visible, Option 1=Collapsed, Index=IsShippingBuild!(self:Fort Runtime Options Object Reference=@n41) -> (ReturnValue:bool))
   )
   n42 = GetRuntimeOptions()
   > SetVisibility(
    self=Get Button_Locker() -> (Button_Locker:Common Button Base Object Reference)
    InVisibility=Select(Option 0=Visible, Option 1=Collapsed, Index=IsShippingBuild(self=@n42))
   )
   > K2Node_Composite_0 = Mobile Game Stuff!(execute) -> (then:Exec) {is=title K2Node_Composite}
   > Handle Low Power Mode()
   > SetSupportButtonVisibility()
   > SetLegalButtonVisibility()
   > Bind Event to On Is Creative Moderator Mode Active Changed(
    self:Creative HUDLContext Object Reference=Get CreativeHUDLContext!() -> (ReturnValue:Creative HUDLContext Object Reference) {is=title K2Node_GetBlueprintContextEx}
    Delegate=Create Event()
   ) {is=title K2Node_AddDelegate}
   > n43 = Cast To FortPlayerControllerGameplay(Object=GetOwningPlayer())
   > Bind Event to On Possessed Pawn Changed(self:Controller Object Reference=@n43, Delegate=Create Event()) {is=title K2Node_AddDelegate}
   > n44 = Handle On Is Creative Moderator Mode Active Changed!(
    execute
    self
    bIsCreativeModeratorModeActive:bool=IsCreativeModeratorModeActive!(self:Fort Player Controller Gameplay Object Reference=@n43) -> (ReturnValue:bool)
   ) -> (then) {is=title K2Node_CallFunction}
   n45 = Get CommonInputSubsystem!() -> (ReturnValue:Common Input Object Reference) {is=title K2Node_GetBlueprintContextEx}
   Bind Event to On Input Method Changed(execute=[@n44.then, @n43.CastFailed], self=@n45, Delegate=Create Event())
   > Handle Input Method Changed!(execute, self, bNewInputType:ECommonInputType Enum=GetCurrentInputType(self=@n45)) -> (then)
   Event Destruct()
   > Bind Event to On Is Creative Moderator Mode Active Changed(self=Get CreativeHUDLContext(), Delegate=Create Event())
   n46 = Event FakeSidebarBackButtonClick() {is=title K2Node_Event}
   Event OnRemovedFromFocusPath!() -> (OutputDelegate, then, InFocusEvent:Focus Event Structure) {is=title K2Node_Event}
   > Set LastSelectedButton()
   Event OnRefreshOptions() {is=title K2Node_Event}
   > UpdateButtons(self:Self Object Reference)
   > SwitchSubMenu(Index=0)
   > Set LastSelectedButton()
   Event On Activated()
   > n47 = Branch(Condition=IsMobileGame())
   > n48 = Sequence!(execute) -> (then_0, then_1)
   > n49 = Branch(Condition=HasActiveProductTag!() -> (ReturnValue:bool))
   n50 = GetRuntimeOptions()
   n51 = GetHUDLayoutTool_AllowedProductTags!(self:Fort Runtime Options Object Reference=@n50) -> (OutTagContainer:Gameplay Tag Container Structure)
   n52 = GetHUDLayoutTool_DisallowedProductTags!(self:Fort Runtime Options Object Reference=@n50) -> (OutTagContainer:Gameplay Tag Container Structure)
   > SetVisibility(
    self=Get ButtonHUDLayoutTool() -> (ButtonHUDLayoutTool:Main Menu Button Object Reference)
    InVisibility=Select(
     Option 0=Collapsed
     Option 1=SelfHitTestInvisible
     Index=IsCurrentContextProductTagAllowedByLists!(
      AllowedTagList:"Gameplay Tag Container Structure (by ref)"=@n51
      DisallowedTagList:"Gameplay Tag Container Structure (by ref)"=@n52
      bMatchDefaultTag:bool=true
     ) -> (ReturnValue:bool)
    )
   )
   n53 = GetRuntimeOptions()
   n54 = Get ControlHelp_AllowedProductTags!(self:Fort Runtime Options Object Reference=@n53) -> (ControlHelp_AllowedProductTags:Gameplay Tag Container Structure)
   n55 = Get ControlHelp_DisallowedProductTags!(self:Fort Runtime Options Object Reference=@n53) -> (ControlHelp_DisallowedProductTags:Gameplay Tag Container Structure)
   > SetVisibility(
    self=Get ButtonBacchusHelp() -> (ButtonBacchusHelp:Main Menu Button Object Reference)
    InVisibility=Select(
     Option 0=Collapsed
     Option 1=Visible
     Index=IsCurrentContextProductTagAllowedByLists(AllowedTagList=@n54, DisallowedTagList=@n55, bMatchDefaultTag=true)
    )
   )
   SetVisibility(execute=@n49.else, self=Get ButtonHUDLayoutTool(), InVisibility=Collapsed)
   > SetVisibility(self=Get ButtonBacchusHelp(), InVisibility=Collapsed)
   Branch(execute=@n48.then_1, Condition=ShouldForceHUDLayoutToolButtonVisible!() -> (ReturnValue:bool))
   > SetVisibility(self=Get ButtonHUDLayoutTool(), InVisibility=SelfHitTestInvisible)
   SetVisibility(execute=@n47.else, self=Get ButtonHUDLayoutTool(), InVisibility=Collapsed)
   > SetVisibility(self=Get ButtonBacchusHelp(), InVisibility=Collapsed)
   UpdateMinigameButtons(execute=@n34.then) {is=title K2Node_CallFunction}
  ]}
 Mobile Game Stuff {
  path=EventGraph.K2Node_Composite_0.Mobile Game Stuff
  nodes [
   Inputs()
   > n1 = Branch(Condition=IsMobileGame())
   n2 = NOT Boolean(A=IsShippingBuild(self=GetRuntimeOptions()))
   n3 = IsMobilePlatform!() -> (ReturnValue:bool)
   > SetVisibility(
    self=Get ConsoleButton() -> (ConsoleButton:Main Menu Button Object Reference)
    InVisibility=Select(Option 0=Collapsed, Option 1=Visible, Index=AND Boolean(A=@n2, B=@n3))
   )
   n4 = NOT Boolean(A=IsShippingBuild(self=GetRuntimeOptions()))
   n5 = NOT Boolean(A=IsMobilePlatform())
   > SetVisibility(self=Get ConsoleCommandPrompt(), InVisibility=Select(Option 0=Collapsed, Option 1=Visible, Index=AND Boolean(A=@n4, B=@n5)))
   n6 = Get HorizontalBox_Header() -> (HorizontalBox_Header:Horizontal Box Object Reference)
   n7 = Get SidebarBackButton() -> (SidebarBackButton:Common Button Base Object Reference)
   > AddChildToHorizontalBox!(execute, self:Horizontal Box Object Reference=@n6, Content:Widget Object Reference=@n7) -> (then, ReturnValue:Horizontal Box Slot Object Reference)
   n8 = Get Text_VoiceChatTitle() -> (Text_VoiceChatTitle:Common Text Object Reference)
   > AddChildToHorizontalBox(self=@n6, Content=@n8)
   > SetPadding!(
    execute
    self:Horizontal Box Slot Object Reference=SlotAsHorizontalBoxSlot!(Widget:Widget Object Reference=@n7) -> (ReturnValue:Horizontal Box Slot Object Reference)
    InPadding:Margin Structure=Make Margin(Left=-24, Top=0, Right=-16, Bottom=0)
   ) -> (then)
   n9 = SlotAsHorizontalBoxSlot(Widget=@n8)
   > SetSize!(
    execute
    self:Horizontal Box Slot Object Reference=@n9
    InSize:Slate Child Size Structure=Make Slate Child Size!(Value:float=1, SizeRule:ESlateSizeRule Enum=Fill) -> (SlateChildSize:Slate Child Size Structure) {is=title K2Node_MakeStruct}
   ) -> (then)
   > n10 = SetVerticalAlignment(self:Horizontal Box Slot Object Reference=@n9, InVerticalAlignment=VAlign_Center)
   Outputs!(then:Exec=[@n10.then, @n1.else])
  ]
 }
 Handle Input Method Changed {nodes [
   Handle Input Method Changed!() -> (then, bNewInputType:ECommonInputType Enum) {is=title K2Node_FunctionEntry}
   n1 = Get FortSettingsContext!() -> (ReturnValue:Fort Settings Context Object Reference) {is=title K2Node_GetBlueprintContextEx}
   > SetVisibility(
    self=Get Button_Settings() -> (Button_Settings:Common Button Legacy Object Reference)
    InVisibility=Select(
     Option 0=SelfHitTestInvisible
     Option 1=Collapsed
     Index=ShouldSeparateSettingsAndMoreMenus!(self:Fort Settings Context Object Reference=@n1) -> (ReturnValue:bool)
    )
   )
   > RefreshFocus(self:Fort Blade Menu Main Menu Object Reference)
  ]}
 Handle Low Power Mode {nodes [
   Handle Low Power Mode() {is=title K2Node_FunctionEntry}
   > Bind Event to On Low Performance Mode(
    self:Athena HUDContext Object Reference=Get AthenaHUDContext!() -> (ReturnValue:Athena HUDContext Object Reference) {is=title K2Node_GetBlueprintContextEx}
    Delegate=Create Event()
   ) {is=title K2Node_AddDelegate}
   > UpdateLowPowerMessage!(execute, self, Passthrough:bool=false) -> (then)
  ]}
 InitializeMainMenu {nodes [
   InitializeMainMenu() {is=title K2Node_FunctionEntry}
   n1 = Get FortGlobalUIContext()
   n2 = AllowQuit!(self:Fort Global UIContext Object Reference=@n1) -> (ReturnValue:bool)
   n3 = NOT Boolean(A=IsInZone(self=@n1))
   > SetVisibility(InVisibility=Select(Option 0=Collapsed, Option 1=SelfHitTestInvisible, Index=AND Boolean(A=@n3, B=@n2)))
   n4 = Get FortGlobalUIContext()
   > SetVisibility(
    self=Get ButtonCredits() -> (ButtonCredits:Main Menu Button Object Reference)
    InVisibility=Select(Option 0=SelfHitTestInvisible, Option 1=Collapsed, Index=IsInZone(self=@n4))
   )
   n5 = NOT Boolean(A=IsShippingBuild(self=GetRuntimeOptions()))
   n6 = AND Boolean(
    A=Get bAllowInGameStore!(self:Fort Runtime Options Object Reference=GetRuntimeOptions()) -> (bAllowInGameStore:bool)
    B=IsInZone(self=Get FortGlobalUIContext())
   )
   > SetVisibility(
    self=Get Button_Store() -> (Button_Store:Common Button Base Object Reference)
    InVisibility=Select(Option 0=Collapsed, Option 1=SelfHitTestInvisible, Index=AND Boolean(A=@n6, B=@n5))
   )
  ]}
 SetLegalButtonVisibility {nodes [
   SetLegalButtonVisibility() {is=title K2Node_FunctionEntry}
   n1 = GetFeatureState!(self:Fort Global UIContext Object Reference=Get FortGlobalUIContext(), Feature:EFortUIFeature Enum=ShowEULA) -> (OutFeatureState:EFortUIFeatureState Enum=Enabled, OutReason:EFortUIFeatureStateReason Enum=Default)
   > SetVisibility(
    self=Get Legal() -> (Legal:Main Menu Button Object Reference)
    InVisibility=Select(Option 0=Collapsed, Option 1=Visible, Index="Equal (Enum)"(A:EFortUIFeatureState Enum=@n1, B:EFortUIFeatureState Enum=Enabled))
   )
  ]}
 SetSupportButtonVisibility {nodes [
   SetSupportButtonVisibility() {is=title K2Node_FunctionEntry}
   n1 = GetFeatureState(self=Get FortGlobalUIContext(), Feature=ShowCustomerSupport) -> (OutFeatureState=Enabled, OutReason=Default)
   > SetVisibility(
    self=Get Button_SupportSubMenu() -> (Button_SupportSubMenu:Main Menu Button Object Reference)
    InVisibility=Select(Option 0=Collapsed, Option 1=Visible, Index="Equal (Enum)"(A=@n1, B=Enabled))
   )
  ]}
 ShowMinigameButtons {nodes [
   n1 = ShowMinigameButtons!() -> (then, Player:Fort Player Controller Athena Object Reference, GameState:Fort Game State Athena Object Reference) {is=title K2Node_FunctionEntry}
   > n2 = Branch(Condition=UseLegacyShowMinigameButtons!() -> (ReturnValue:bool))
   > Return Node!(
    execute
    bResult:bool=ShowMinigameButtonsLegacy!(Player:Fort Player Controller Athena Object Reference=@n1, FortGameStateAthena:Fort Game State Athena Object Reference=@n1.GameState) -> (bResult:bool)
   )
   Return Node(
    execute=@n2.else
    bResult=ShowMinigameButtonsBase!(PlayerController:Fort Player Controller Athena Object Reference=@n1, GameState:Fort Game State Athena Object Reference=@n1.GameState) -> (ReturnValue:bool)
   )
  ]}
 ShowMinigameButtonsLegacy {nodes [
   ShowMinigameButtonsLegacy!() -> (then, Player:Fort Player Controller Athena Object Reference, FortGameStateAthena:Fort Game State Athena Object Reference) {is=title K2Node_FunctionEntry}
   > n1 = Sequence()
   > Get!(execute:Exec) -> (then:Exec, else:Exec, Player:Fort Player Controller Athena Object Reference)
   > Set HasPermissionToStart(HasPermissionToStart:bool=CanOperateOnMinigame!(PlayerController:Fort Player Controller Object Reference=Get Player() -> (Player:Fort Player Controller Athena Object Reference)) -> (ReturnValue:bool)) -> (Output_Get:bool)
   > n2 = Set Minigame(Minigame:Fort Minigame Object Reference=GetMinigame!(ContextActor:Actor Object Reference=GetOwningPlayer()) -> (ReturnValue)) -> (Output_Get:Fort Minigame Object Reference)
   > Is Valid(InputObject=@n2)
   n3 = Get Player()
   > n4 = Set bIsModerator(bIsModerator:bool=IsCreativeModeratorModeEnabled!(self:Fort Player Controller Gameplay Object Reference=@n3) -> (ReturnValue:bool)) -> (Output_Get:bool)
   n5 = HasStarted!(self:Fort Minigame Object Reference=@n2) -> (ReturnValue:bool)
   > Branch(Condition=OR Boolean!(A=@n4, B=CanOperateOnMinigame(PlayerController=@n3), C:bool=NOT Boolean(A=@n5)) -> (ReturnValue))
   n6 = IsCuratedHub!(self:Fort Minigame Object Reference=Get Minigame() -> (Minigame:Fort Minigame Object Reference)) -> (ReturnValue:bool)
   n7 = NOT Boolean(A=IsInPublicMatchmadeCreativeMinigame!(self:Fort Player Controller Athena Object Reference=Get Player()) -> (ReturnValue:bool))
   n8 = Get bIsModerator() -> (bIsModerator:bool)
   > Set bShow(bShow:bool=AND Boolean!(A=OR Boolean!(A=@n8, B=@n7) -> (ReturnValue), B=NOT Boolean(A=@n6), C:bool=Get HasPermissionToStart() -> (HasPermissionToStart:bool)) -> (ReturnValue)) -> (Output_Get:bool)
   Return Node(execute=@n1.then_1, bResult=Get bShow() -> (bShow:bool))
  ]}
 SwitchSubMenu {nodes [
   SwitchSubMenu!() -> (then, Index:int) {is=title K2Node_FunctionEntry}
   n1 = Get Index() -> (Index:int=0)
   > SetActiveWidgetIndex(self=Get Switcher_Menus(), Index=@n1)
   > SetVisibility(self=Get SidebarBackButton(), InVisibility=Select(Option 0=Visible, Option 1=Collapsed, Index="Equal (Integer)"(A=@n1)))
   > RefreshFocus()
   > Set LastSelectedButton()
   > "SetText (Text)"(
    self=Get Text_VoiceChatTitle()
    InText=Select!(
     Option 0:Text="LOCTEXT(\"Menu\")"
     Option 1:Text="LOCTEXT(\"Reporting\")"
     Option 2:Text="LOCTEXT(\"Support && Refunds\")"
     Option 3:Text="LOCTEXT(\"Legal\")"
     Index:int=Get Index() -> (Index=0)
    ) -> (ReturnValue:Text)
   )
   > Branch(Condition="Equal (Integer)"(A=Get Index() -> (Index=0), B=2))
   n2 = GetOwningPlayer()
   n3 = Cast To FortPlayerControllerZone!(Object=@n2) -> (AsFort Player Controller Zone:Fort Player Controller Zone Object Reference, bSuccess:bool=~) {is=title K2Node_DynamicCast}
   n4 = Cast To FortReplaySpectator!(Object=@n2) -> (AsFort Replay Spectator:Fort Replay Spectator Object Reference, bSuccess:bool=~) {is=title K2Node_DynamicCast}
   > SetVisibility(
    self=Get ReturnOrCancelPurchase() -> (ReturnOrCancelPurchase:Main Menu Button Object Reference)
    InVisibility=Select!(
     Option 0:ESlateVisibility Enum=SelfHitTestInvisible
     Option 1:ESlateVisibility Enum=Collapsed
     Index:bool=OR Boolean(A=@n3.bSuccess, B=@n4.bSuccess)
    ) -> (ReturnValue:ESlateVisibility Enum)
   )
  ]}
 Update Team Select Visibility {nodes [
   Update Team Select Visibility() {is=title K2Node_FunctionEntry}
   n1 = GetOwningPlayer()
   > n2 = Cast To FortPlayerControllerAthena(Object=@n1)
   > n3 = Cast To FortGameStateAthena!(execute, Object=GetGameState!() -> (ReturnValue:Game State Base Object Reference)) -> (then, CastFailed, AsFort Game State Athena:Fort Game State Athena Object Reference) {is=title K2Node_DynamicCast}
   n4 = IsCreativeModeratorModeActive(self=@n2)
   n5 = Get Controlled Pawn!(self:Controller Object Reference=@n1) -> (ReturnValue:Pawn Object Reference) {is=title K2Node_CallFunction}
   n6 = AND Boolean(
    A=IsTeamSwitchAllowed!(self:Fort Game State Athena Object Reference=@n3, SwitchingPawn:Pawn Object Reference=@n5) -> (ReturnValue:bool)
    B=NOT Boolean(A=Cast To FortRemoteControlledPawnAthena!(Object=@n5) -> (AsFort Remote Controlled Pawn Athena:Fort Remote Controlled Pawn Athena Object Reference, bSuccess:bool=~) {is=title K2Node_DynamicCast}.bSuccess)
    C=GetShowTeamSelectButton!() -> (ReturnValue:bool)
   )
   > SetVisibility(
    self=Get TeamSelectButton() -> (TeamSelectButton:Main Menu Button Object Reference)
    InVisibility=Select(Option 0=Collapsed, Option 1=SelfHitTestInvisible, Index=OR Boolean(A=@n6, B=@n4))
   )
   SetVisibility(execute=[@n3.CastFailed, @n2.CastFailed], self=Get TeamSelectButton(), InVisibility=Collapsed)
  ]}
 UpdateButtons {nodes [
   UpdateButtons() {is=title K2Node_FunctionEntry}
   > n1 = Sequence!(execute) -> (then_0, then_1, then_2:Exec, then_3:Exec, then_4:Exec)
   n2 = Set CurrentMinigame(execute=@n1.then_4, CurrentMinigame:Fort Minigame Object Reference=GetMinigame(ContextActor=GetOwningPlayer())) -> (Output_Get:Fort Minigame Object Reference)
   > Branch(Condition="Not Equal (Object)"(A:Object Reference=@n2, B:Object Reference=Get LastMinigame() -> (LastMinigame:Fort Minigame Object Reference)) -> (ReturnValue:bool=~) {is=title K2Node_PromotableOperator})
   > Set LastMinigame(LastMinigame:Fort Minigame Object Reference=Get CurrentMinigame() -> (CurrentMinigame:Fort Minigame Object Reference)) -> (Output_Get:Fort Minigame Object Reference)
   n3 = GetMinigame(ContextActor=GetOwningPlayer())
   Is Valid(exec=@n1.then_1, InputObject=@n3)
   > Set Machine(Machine:Fort Minigame Settings Building Object Reference=GetMinigameSettingsMachine!(ContextActor:Actor Object Reference=@n3) -> (ReturnValue:Fort Minigame Settings Building Object Reference)) -> (Output_Get:Fort Minigame Settings Building Object Reference)
   UpdateMinigameButtons(execute=@n1.then_2)
   > Bind To State Change Delegate() {is=title K2Node_CallFunction}
   > n4 = Cast To MinigameButtons(Object=Get MinigameButtons())
   > SetWidgetSwitcherIndex!(execute, self:Minigame Buttons Object Reference=@n4, Index:int=1) -> (then)
   UpdateLowPowerMessage(execute=@n1.then_0, Passthrough=false)
   > Update Team Select Visibility() {is=title K2Node_CallFunction}
   > UpdateRespawnWidgetVisibility()
   > UpdateCreativeButtonsVisibility()
   OpenSettingsIfSaved(execute=@n1.then_3, self:Fort Settings Context Object Reference=Get FortSettingsContext())
  ]}
 UpdateCreativeButtonsVisibility {nodes [
   UpdateCreativeButtonsVisibility() {is=title K2Node_FunctionEntry}
   > n1 = Cast To FortGameStateAthena(Object=GetGameState())
   > Set Local_GameState(Local_GameState:Fort Game State Athena Object Reference=@n1) -> (Output_Get:Fort Game State Athena Object Reference)
   > n2 = Cast To FortPlayerControllerAthena(Object=GetOwningPlayer())
   > n3 = Set Local_PlayerController(Local_PlayerController:Fort Player Controller Athena Object Reference=@n2) -> (Output_Get:Fort Player Controller Athena Object Reference)
   > n4 = Cast To FortPlayerState!(execute, Object=Get PlayerState!(self:Controller Object Reference=@n3) -> (PlayerState:Player State Object Reference)) -> (then, CastFailed, AsFort Player State:Fort Player State Object Reference) {is=title K2Node_DynamicCast}
   > Set Local_PlayerState(Local_PlayerState:Fort Player State Object Reference=@n4) -> (Output_Get:Fort Player State Object Reference)
   n5 = Get Local_PlayerController() -> (Local_PlayerController:Fort Player Controller Athena Object Reference)
   n6 = Get Button_MyIsland() -> (Button_MyIsland:Main Menu Button Object Reference)
   > SetVisibility(
    self=@n6
    InVisibility=Select(
     Option 0=Collapsed
     Option 1=SelfHitTestInvisible
     Index=IsPlayerInAnExperienceTheyOwn!(self:Fort Player Controller Object Reference=@n5) -> (ReturnValue:bool)
    )
   )
   > SetText(self:Fort Main Menu Button Object Reference=@n6, InText=GetMyIslandDisplayName!() -> (ReturnValue:"Text (by ref)"))
   n7 = Get Local_GameState() -> (Local_GameState:Fort Game State Athena Object Reference)
   n8 = Get Local_PlayerController()
   > SetVisibility(
    self=Get MinigameButtons()
    InVisibility=Select(
     Option 0=Collapsed
     Option 1=SelfHitTestInvisible
     Index=ShowMinigameButtons!(Player:Fort Player Controller Athena Object Reference=@n8, GameState:Fort Game State Athena Object Reference=@n7) -> (bResult:bool) {is=title K2Node_CallFunction}
    )
   )
   n9 = Get Local_GameState()
   n10 = IsTeleportToCreativeHubAllowed!(
    self:Fort Game State Athena Object Reference=@n9
    PlayerState:Fort Player State Object Reference=Get Local_PlayerState() -> (Local_PlayerState:Fort Player State Object Reference)
   ) -> (ReturnValue:bool)
   n11 = CanRespawnOnStarterIsland!(self:Fort Game State Athena Object Reference=@n9) -> (ReturnValue:bool)
   > SetVisibility(
    self=Get Button_BackToHub() -> (Button_BackToHub:Common Button Legacy Object Reference)
    InVisibility=Select(Option 0=Collapsed, Option 1=SelfHitTestInvisible, Index=AND Boolean!(A=@n11, B=@n10) -> (ReturnValue))
   )
   n12 = NOT Boolean(A=IsMobileGame())
   n13 = Get Local_PlayerController()
   n14 = IsCreativeModeratorModeEnabled(self=@n13)
   n15 = IsCreativeModeratorModeAllowed!(self:Fort Player Controller Gameplay Object Reference=@n13) -> (ReturnValue:bool)
   > SetVisibility(
    self=Get ModeratorMode() -> (ModeratorMode:Overlay Object Reference)
    InVisibility=Select(Option 0=Collapsed, Option 1=SelfHitTestInvisible, Index=AND Boolean!(A=@n12, B=@n15, C:bool=@n14) -> (ReturnValue))
   )
   n16 = GetMinigame(ContextActor=Get Local_PlayerController())
   > n17 = Is Valid(InputObject=@n16)
   > n18 = Branch(Condition=NOT Boolean(A=IsCuratedHub(self=@n16)))
   n19 = IsCreativeIslandExporterEnabled!(self:Fort Player Controller Gameplay Object Reference=Get Local_PlayerController()) -> (ReturnValue:bool)
   n20 = Get Local_GameState()
   n21 = OR Boolean!(
    A=IsCreativeLTMPlaylist!(self:Fort Game State Athena Object Reference=@n20) -> (ReturnValue:bool)
    B=IsCreativeModePlaylist!(self:Fort Game State Athena Object Reference=@n20) -> (ReturnValue:bool)
    C:bool=IsCreativeMMPlaylist!(self:Fort Game State Athena Object Reference=@n20) -> (ReturnValue:bool)
   ) -> (ReturnValue)
   n22 = Get bAllowIslandExporting(self=GetRuntimeOptions())
   > SetVisibility(
    self=Get ExportIslandButton() -> (ExportIslandButton:Main Menu Button Object Reference)
    InVisibility=Select(Option 0=Collapsed, Option 1=SelfHitTestInvisible, Index=AND Boolean(A=@n19, B=@n22, C=@n21))
   )
   n23 = Get Controlled Pawn(self=Get Local_PlayerController())
   n24 = Get Local_GameState()
   > SetVisibility(
    self=Get AthenaMainMenuClearPlayerDataButton() -> (AthenaMainMenuClearPlayerDataButton:Main Menu Button Object Reference)
    InVisibility=Select(
     Option 0=Collapsed
     Option 1=SelfHitTestInvisible
     Index=IsPlayerClearPersistanceIslandDataAllowed!(self:Fort Game State Athena Object Reference=@n24, SwitchingPawn:Pawn Object Reference=@n23) -> (ReturnValue:bool)
    )
   )
   SetVisibility(execute=[@n18.else, @n17.Is Not Valid], self=[Get ExportIslandButton(), Get AthenaMainMenuClearPlayerDataButton()], InVisibility=Collapsed)
   SetVisibility(
    execute=[@n4.CastFailed, @n2.CastFailed, @n1.CastFailed]
    self=[Get MinigameButtons(), Get ExportIslandButton(), Get AthenaMainMenuClearPlayerDataButton(), Get Button_BackToHub(), Get Button_MyIsland()]
    InVisibility=Collapsed
   )
  ]}
 UpdateLowPowerMessage {nodes [
   UpdateLowPowerMessage!() -> (then, Passthrough:bool=~) {is=title K2Node_FunctionEntry}
   n1 = Get AthenaHUDContext()
   > SetVisibility(
    self=Get TextLowPowerMode() -> (TextLowPowerMode:Common Text Object Reference)
    InVisibility=Select(
     Option 0=Collapsed
     Option 1=SelfHitTestInvisible
     Index=GetInLowPerformanceMode!(self:Athena HUDContext Object Reference=@n1) -> (ReturnValue:bool)
    )
   )
  ]}
 UpdateRespawnWidgetVisibility {nodes [
   UpdateRespawnWidgetVisibility() {is=title K2Node_FunctionEntry}
   > n1 = Cast To FortGameStateAthena(Object=GetGameState())
   > n2 = Cast To FortRemoteControlledPawnAthena!(execute:Exec, Object=GetOwningPlayerPawn!(self:User Widget Object Reference) -> (ReturnValue:Pawn Object Reference)) -> (then:Exec, CastFailed:Exec, AsFort Remote Controlled Pawn Athena)
   SetVisibility(
    execute=[@n5.CastFailed, @n4.Is Not Valid, @n2.then, @n1.CastFailed]
    self=Get RespawnButton() -> (RespawnButton:Main Menu Button Object Reference)
    InVisibility=Collapsed
   )
   n3 = GetOwningPlayer()
   n4 = Is Valid(exec=@n2.CastFailed, InputObject=@n3)
   > n5 = Cast To FortPlayerState(Object=Get PlayerState(self=@n3))
   > Set PlayerTriggeredSpawnAllowed(PlayerTriggeredSpawnAllowed:bool=IsPlayerTriggeredRespawnAllowed!(self:Fort Game State Athena Object Reference=@n1, PlayerState:Fort Player State Object Reference=@n5) -> (ReturnValue:bool)) -> (Output_Get:bool)
   n6 = IsRespawningAllowed!(self:Fort Game State Zone Object Reference=@n1, PlayerState:Fort Player State Object Reference=@n5) -> (ReturnValue:bool)
   n7 = Get PlayerTriggeredSpawnAllowed() -> (PlayerTriggeredSpawnAllowed:bool)
   > SetVisibility(
    self=Get RespawnButton()
    InVisibility=Select(Option 0=Collapsed, Option 1=SelfHitTestInvisible, Index=AND Boolean!(A=@n6, B=@n7) -> (ReturnValue))
   )
  ]}
]
variables [
 CreditsScreenClass {type=Class, default=/FortniteCoreUI/Frontend/Credits/Credits.Credits_C}
 LastMinigame {type=FortMinigame, default=null}
 LastSelectedButton {type=CommonButtonLegacy, default=null}
 LegalInfoClass {type=Class, default=/Game/UI/Foundation/Legal/LegalInfo.LegalInfo_C}
 Machine {type=FortMinigameSettingsBuilding, default=null}
 MobileControlsHelpClass {type=Class, default=/FortniteCoreUI/Frontend/BacchusControlsHelp.BacchusControlsHelp_C}
 ParentalControlsClass_New {type=Class, default=/FortniteCoreUI/UI/ContentControls/ParentalControlsScreen.ParentalControlsScreen_C}
 Player Clear Persistence Data in Progress Description {type=string, default="Do you want to clear your saved data? (You will be respawned as part of the process)"}
 Player Clear Persistence Data Pregame Description {type=string, default=Do you want to clear your saved data?}
 PlayerClearIslandDataClass {type=Class, default=/FortniteCoreUI/PlayerClearPersistenceIslandDataPromptWidget.PlayerClearPersistenceIslandDataPromptWidget_C}
 ReportButtonText_Creative {type=string, default="Report a bug, player, or island"}
 ReportButtonText_Default {type=string, default=Report a bug or player}
 RespawnClass {type=Class, default=/FortniteCoreUI/Frontend/RespawnPromptWidget.RespawnPromptWidget_C}
 SettingTabToOpen {type=string, default=None}
 TeamSelectWidgetClass {type=Class, default=/FortniteCoreUI/Frontend/TeamSelect/TeamSelectWidget.TeamSelectWidget_C}
]
widgets {
 info {inheritedWidgetCount=19, parentClass=/Script/FortniteUI.FortBladeMenu_MainMenu, rootWidgetClass=&Border, widgetCount=65}
 widgets [
  Border_Content {
   widgetClassPath=&Border
   widget=Border_Content
   properties {
    navigation=&self:WidgetTree.Border_Content.WidgetNavigation_0
    background {
     bIsDynamicallyLoaded=false
     drawAs=NoDrawType
     imageType=NoImage
     mirroring=NoMirror
     resourceName=None
     resourceObject=null
     tiling=NoTile
     imageSize=<32 32>
     margin=<ltrb 0 0 0 0>
     outlineSettings {
      bUseBrushTransparency=false
      roundingType=HalfHeightRadius
      width=0
      color {colorUseRule=UseColor_Specified, specifiedColor=<rgba 0 0 0 0>}
      cornerRadii=<0 0 0 0>
     }
     tintColor {colorUseRule=UseColor_Specified, specifiedColor=<rgba 1 1 1 1>}
     uVRegion {bIsValid=false, max=<0 0>, min=<0 0>}
    }
    padding=<ltrb 0 0 0 0>
   }
  }
  Overlay_1 {widgetClassPath=&Overlay, parent=Border_Content, widget=Overlay_1, slot {class=BorderSlot, properties {padding=<ltrb 0 0 0 0>}}}
  SafeZone_Content {
   widgetClassPath=&SafeZone
   bIsVariable=true
   parent=Overlay_1
   widget=SafeZone_Content
   properties {padLeft=false, padRight=false, padTop=false}
   slot {class=OverlaySlot, properties {horizontalAlignment=HAlign_Fill, verticalAlignment=VAlign_Fill}}
  }
  Overlay_0 {widgetClassPath=&Overlay, parent=SafeZone_Content, widget=Overlay_0, slot {class=SafeZoneSlot}}
  VerticalBox_0 {
   widgetClassPath=&VerticalBox
   parent=Overlay_0
   widget=VerticalBox_0
   slot {class=OverlaySlot, properties {horizontalAlignment=HAlign_Fill, verticalAlignment=VAlign_Fill}}
  }
  Border_1 {
   widgetClassPath=&Border
   parent=VerticalBox_0
   widget=Border_1
   properties {
    background {
     bIsDynamicallyLoaded=false
     drawAs=Image
     imageType=NoImage
     mirroring=NoMirror
     resourceName=None
     resourceObject=/Game/UI/Sidebar/Materials/MI_UI_Sidebar_Header_BaseFade
     tiling=NoTile
     imageSize=<4096 4096>
     margin=<ltrb 0 0 0 0>
     outlineSettings {
      bUseBrushTransparency=false
      roundingType=HalfHeightRadius
      width=0
      color {colorUseRule=UseColor_Specified, specifiedColor=<rgba 0 0 0 0>}
      cornerRadii=<0 0 0 0>
     }
     tintColor {colorUseRule=UseColor_Specified, specifiedColor=<rgba 1 1 1 1>}
     uVRegion {bIsValid=false, max=<0 0>, min=<0 0>}
    }
    padding=<ltrb 20 60 20 20>
   }
   slot {class=VerticalBoxSlot, properties {verticalAlignment=VAlign_Top}}
  }
  HorizontalBox_Header {
   widgetClassPath=&HorizontalBox
   bIsVariable=true
   parent=Border_1
   widget=HorizontalBox_Header
   slot {class=BorderSlot, properties {padding=<ltrb 20 60 20 20>}}
  }
  Text_VoiceChatTitle {
   widgetClassPath=&CommonTextBlock
   bIsVariable=true
   parent=HorizontalBox_Header
   widget=Text_VoiceChatTitle
   properties {
    autoWrapText=true
    text=Menu
    textTransformPolicy=ToUpper
    font {
     bForceMonospaced=false
     bMaterialIsStencil=false
     fontMaterial=null
     fontObject=&BurbankSmall
     letterSpacing=0
     monospacedWidth=1
     size=32
     skewAmount=0
     typefaceFontName=Black
     outlineSettings {
      bApplyOutlineToDropShadows=false
      bMiteredCorners=false
      bSeparateFillAlpha=false
      outlineMaterial=null
      outlineSize=0
      outlineColor=<rgba 0 0 0 1>
     }
    }
    renderTransform {angle=0, scale=<1 1>, shear=<-8 0>, translation=<0 0>}
    shadowOffset=<0 0>
   }
   slot {class=HorizontalBoxSlot, properties {verticalAlignment=VAlign_Center, size {sizeRule=Fill, value=1}}}
  }
  CommonHierarchicalScrollBox_0 {
   widgetClassPath=/Script/CommonUI.CommonHierarchicalScrollBox
   parent=VerticalBox_0
   widget=CommonHierarchicalScrollBox_0
   properties {navigation=&self:WidgetTree.CommonHierarchicalScrollBox_0.WidgetNavigation_1}
   slot {class=VerticalBoxSlot, properties {padding=<ltrb 20 12 20 40>, size {sizeRule=Fill, value=1}}}
  }
  Switcher_Menus {
   widgetClassPath=&WidgetSwitcher
   bInherited=true
   bIsVariable=true
   parent=CommonHierarchicalScrollBox_0
   widget=Switcher_Menus
   slot {class=ScrollBoxSlot}
  }
  VBox_MainMenuButtons {
   widgetClassPath=&VerticalBox
   bIsVariable=true
   parent=Switcher_Menus
   widget=VBox_MainMenuButtons
   properties {flowDirectionPreference=Culture}
   slot {class=WidgetSwitcherSlot}
  }
  Button_Settings {
   widgetClassPath=/FortniteCoreUI/UI/Frontend/MainMenu/MainMenuButton_Large
   bInherited=true
   bIsVariable=true
   parent=VBox_MainMenuButtons
   widget=Button_Settings
   properties {
    defaultText=Settings
    onButtonBaseClicked=(&self:WidgetTree.Button_Settings.HandleOnButtonClicked)
    onButtonBaseDoubleClicked=(&self:WidgetTree.Button_Settings.HandleOnButtonDoubleClicked)
    onButtonBaseHovered=(&self:WidgetTree.Button_Settings.HandleOnButtonHovered)
    onButtonBaseUnhovered=(&self:WidgetTree.Button_Settings.HandleOnButtonUnhovered)
    onSelectedChangedBase=(&self:WidgetTree.Button_Settings.HandleOnSelectedChanged)
    defaultBrush {
     bIsDynamicallyLoaded=false
     drawAs=Image
     imageType=NoImage
     mirroring=NoMirror
     resourceName=None
     resourceObject=/Game/UI/SocialPanel/Art/Textures/T_UI_Icon_LocalUserSettings
     tiling=NoTile
     imageSize=<48 48>
     margin=<ltrb 0 0 0 0>
     outlineSettings {
      bUseBrushTransparency=false
      roundingType=HalfHeightRadius
      width=0
      color {colorUseRule=UseColor_Specified, specifiedColor=<rgba 0 0 0 0>}
      cornerRadii=<0 0 0 0>
     }
     tintColor {colorUseRule=UseColor_Specified, specifiedColor=<rgba 0.8517 0.9344 0.9911 1>}
     uVRegion {bIsValid=false, max=<0 0>, min=<0 0>}
    }
    textPadding=<ltrb 8 0 0 0>
   }
   slot {class=VerticalBoxSlot, properties {padding=<ltrb 0 8 0 0>}}
  }
  Button_Subscription {
   widgetClassPath=&MainMenuButton_Large
   bInherited=true
   bIsVariable=true
   parent=VBox_MainMenuButtons
   widget=Button_Subscription
   properties {
    defaultText=Fortnite Crew
    onButtonBaseClicked=(&self:WidgetTree.Button_Subscription.HandleOnButtonClicked)
    onButtonBaseDoubleClicked=(&self:WidgetTree.Button_Subscription.HandleOnButtonDoubleClicked)
    onButtonBaseHovered=(&self:WidgetTree.Button_Subscription.HandleOnButtonHovered)
    onButtonBaseUnhovered=(&self:WidgetTree.Button_Subscription.HandleOnButtonUnhovered)
    onSelectedChangedBase=(&self:WidgetTree.Button_Subscription.HandleOnSelectedChanged)
    defaultBrush {
     bIsDynamicallyLoaded=false
     drawAs=Image
     imageType=NoImage
     mirroring=NoMirror
     resourceName=None
     resourceObject=/Game/UI/SocialPanel/Art/Textures/T_UI_Icon_FortniteCrew
     tiling=NoTile
     imageSize=<48 48>
     margin=<ltrb 0 0 0 0>
     outlineSettings {
      bUseBrushTransparency=false
      roundingType=HalfHeightRadius
      width=0
      color {colorUseRule=UseColor_Specified, specifiedColor=<rgba 0 0 0 0>}
      cornerRadii=<0 0 0 0>
     }
     tintColor {colorUseRule=UseColor_Specified, specifiedColor=<rgba 0.8517 0.9344 0.9911 1>}
     uVRegion {bIsValid=false, max=<0 0>, min=<0 0>}
    }
    textPadding=<ltrb 8 0 0 0>
   }
   slot {class=VerticalBoxSlot, properties {padding=<ltrb 0 8 0 0>}}
  }
  Spacer_169 {widgetClassPath=&Spacer, parent=VBox_MainMenuButtons, widget=Spacer_169, properties {size=<1 16>}, slot {class=VerticalBoxSlot}}
  ParentalControlsButton {
   widgetClassPath=/Game/UI/Frontend/MainMenu/MainMenuButton
   bIsVariable=true
   parent=VBox_MainMenuButtons
   widget=ParentalControlsButton
   properties {
    defaultText=Parental Controls
    mobileButtonHeight=80
    onButtonBaseClicked=(&self:WidgetTree.ParentalControlsButton.HandleOnButtonClicked)
    onButtonBaseDoubleClicked=(&self:WidgetTree.ParentalControlsButton.HandleOnButtonDoubleClicked)
    onButtonBaseHovered=(&self:WidgetTree.ParentalControlsButton.HandleOnButtonHovered)
    onButtonBaseUnhovered=(&self:WidgetTree.ParentalControlsButton.HandleOnButtonUnhovered)
    onSelectedChangedBase=(&self:WidgetTree.ParentalControlsButton.HandleOnSelectedChanged)
    style=/Game/UI/Sidebar/Sidebar_ButtonStyle-Tab-Main-Recolor.Sidebar_ButtonStyle-Tab-Main-Recolor_C
   }
   slot {class=VerticalBoxSlot, properties {padding=<ltrb 0 8 0 0>}}
  }
  Button_ManageUnlockedContent {
   widgetClassPath=&MainMenuButton
   bInherited=true
   bIsVariable=true
   parent=VBox_MainMenuButtons
   widget=Button_ManageUnlockedContent
   properties {
    defaultText=Squeegee
    mobileButtonHeight=80
    onButtonBaseClicked=(&self:WidgetTree.Button_ManageUnlockedContent.HandleOnButtonClicked)
    onButtonBaseDoubleClicked=(&self:WidgetTree.Button_ManageUnlockedContent.HandleOnButtonDoubleClicked)
    onButtonBaseHovered=(&self:WidgetTree.Button_ManageUnlockedContent.HandleOnButtonHovered)
    onButtonBaseUnhovered=(&self:WidgetTree.Button_ManageUnlockedContent.HandleOnButtonUnhovered)
    onSelectedChangedBase=(&self:WidgetTree.Button_ManageUnlockedContent.HandleOnSelectedChanged)
    style=&Sidebar_ButtonStyle-Tab-Main-Recolor_C
   }
   slot {class=VerticalBoxSlot, properties {padding=<ltrb 0 8 0 0>}}
  }
  Button_Cilantro {
   widgetClassPath=&MainMenuButton
   bInherited=true
   bIsVariable=true
   parent=VBox_MainMenuButtons
   widget=Button_Cilantro
   properties {
    defaultText=Cilantro
    mobileButtonHeight=80
    onButtonBaseClicked=()
    onButtonBaseDoubleClicked=()
    onButtonBaseHovered=()
    onButtonBaseUnhovered=()
    onSelectedChangedBase=()
    style=&Sidebar_ButtonStyle-Tab-Main-Recolor_C
   }
   slot {class=VerticalBoxSlot, properties {padding=<ltrb 0 8 0 0>}}
  }
  NewPlayerFeedback {
   widgetClassPath=&MainMenuButton
   bIsVariable=true
   parent=VBox_MainMenuButtons
   widget=NewPlayerFeedback
   properties {
    defaultText=Reporting
    mobileButtonHeight=80
    onButtonBaseClicked=(&self:WidgetTree.NewPlayerFeedback.HandleOnButtonClicked)
    onButtonBaseDoubleClicked=(&self:WidgetTree.NewPlayerFeedback.HandleOnButtonDoubleClicked)
    onButtonBaseHovered=(&self:WidgetTree.NewPlayerFeedback.HandleOnButtonHovered)
    onButtonBaseUnhovered=(&self:WidgetTree.NewPlayerFeedback.HandleOnButtonUnhovered)
    onSelectedChangedBase=(&self:WidgetTree.NewPlayerFeedback.HandleOnSelectedChanged)
    style=&Sidebar_ButtonStyle-Tab-Main-Recolor_C
   }
   slot {class=VerticalBoxSlot, properties {padding=<ltrb 0 8 0 0>}}
  }
  Button_SupportSubMenu {
   widgetClassPath=&MainMenuButton
   bIsVariable=true
   parent=VBox_MainMenuButtons
   widget=Button_SupportSubMenu
   properties {
    defaultText=Support and Refunds
    mobileButtonHeight=80
    onButtonBaseClicked=(&self:WidgetTree.Button_SupportSubMenu.HandleOnButtonClicked)
    onButtonBaseDoubleClicked=(&self:WidgetTree.Button_SupportSubMenu.HandleOnButtonDoubleClicked)
    onButtonBaseHovered=(&self:WidgetTree.Button_SupportSubMenu.HandleOnButtonHovered)
    onButtonBaseUnhovered=(&self:WidgetTree.Button_SupportSubMenu.HandleOnButtonUnhovered)
    onSelectedChangedBase=(&self:WidgetTree.Button_SupportSubMenu.HandleOnSelectedChanged)
    style=&Sidebar_ButtonStyle-Tab-Main-Recolor_C
   }
   slot {class=VerticalBoxSlot, properties {padding=<ltrb 0 8 0 0>}}
  }
  Spacer_1 {widgetClassPath=&Spacer, parent=VBox_MainMenuButtons, widget=Spacer_1, properties {size=<1 16>}, slot {class=VerticalBoxSlot}}
  Legal {
   widgetClassPath=&MainMenuButton
   bIsVariable=true
   parent=VBox_MainMenuButtons
   widget=Legal
   properties {
    defaultText=Legal
    mobileButtonHeight=80
    onButtonBaseClicked=(&self:WidgetTree.Legal.HandleOnButtonClicked)
    onButtonBaseDoubleClicked=(&self:WidgetTree.Legal.HandleOnButtonDoubleClicked)
    onButtonBaseHovered=(&self:WidgetTree.Legal.HandleOnButtonHovered)
    onButtonBaseUnhovered=(&self:WidgetTree.Legal.HandleOnButtonUnhovered)
    onSelectedChangedBase=(&self:WidgetTree.Legal.HandleOnSelectedChanged)
    style=&Sidebar_ButtonStyle-Tab-Main-Recolor_C
   }
   slot {class=VerticalBoxSlot, properties {padding=<ltrb 0 8 0 0>}}
  }
  Button_Store {
   widgetClassPath=&MainMenuButton
   bInherited=true
   bIsVariable=true
   parent=VBox_MainMenuButtons
   widget=Button_Store
   properties {
    defaultText="Item Shop - (Under Development)"
    mobileButtonHeight=80
    onButtonBaseClicked=(&self:WidgetTree.Button_Store.HandleOnButtonClicked)
    onButtonBaseDoubleClicked=(&self:WidgetTree.Button_Store.HandleOnButtonDoubleClicked)
    onButtonBaseHovered=(&self:WidgetTree.Button_Store.HandleOnButtonHovered)
    onButtonBaseUnhovered=(&self:WidgetTree.Button_Store.HandleOnButtonUnhovered)
    onSelectedChangedBase=(&self:WidgetTree.Button_Store.HandleOnSelectedChanged)
    style=&Sidebar_ButtonStyle-Tab-Main-Recolor_C
   }
   slot {class=VerticalBoxSlot, properties {padding=<ltrb 0 8 0 0>}}
  }
  ButtonBacchusHelp {
   widgetClassPath=&MainMenuButton
   bIsVariable=true
   parent=VBox_MainMenuButtons
   widget=ButtonBacchusHelp
   properties {
    defaultText=Controls Help
    mobileButtonHeight=80
    onButtonBaseClicked=(&self:WidgetTree.ButtonBacchusHelp.HandleOnButtonClicked)
    onButtonBaseDoubleClicked=(&self:WidgetTree.ButtonBacchusHelp.HandleOnButtonDoubleClicked)
    onButtonBaseHovered=(&self:WidgetTree.ButtonBacchusHelp.HandleOnButtonHovered)
    onButtonBaseUnhovered=(&self:WidgetTree.ButtonBacchusHelp.HandleOnButtonUnhovered)
    onSelectedChangedBase=(&self:WidgetTree.ButtonBacchusHelp.HandleOnSelectedChanged)
    style=/Game/UI/Sidebar/Sidebar_ButtonStyle-Tab-Main-Recolor-Green.Sidebar_ButtonStyle-Tab-Main-Recolor-Green_C
    visibility=Collapsed
   }
   slot {class=VerticalBoxSlot, properties {padding=<ltrb 0 8 0 0>}}
  }
  ButtonHUDLayoutTool {
   widgetClassPath=&MainMenuButton
   bIsVariable=true
   parent=VBox_MainMenuButtons
   widget=ButtonHUDLayoutTool
   properties {
    defaultText=HUD Layout Tool
    mobileButtonHeight=80
    onButtonBaseClicked=(&self:WidgetTree.ButtonHUDLayoutTool.HandleOnButtonClicked)
    onButtonBaseDoubleClicked=(&self:WidgetTree.ButtonHUDLayoutTool.HandleOnButtonDoubleClicked)
    onButtonBaseHovered=(&self:WidgetTree.ButtonHUDLayoutTool.HandleOnButtonHovered)
    onButtonBaseUnhovered=(&self:WidgetTree.ButtonHUDLayoutTool.HandleOnButtonUnhovered)
    onSelectedChangedBase=(&self:WidgetTree.ButtonHUDLayoutTool.HandleOnSelectedChanged)
    style=&Sidebar_ButtonStyle-Tab-Main-Recolor_C
    visibility=Collapsed
   }
   slot {class=VerticalBoxSlot, properties {padding=<ltrb 0 8 0 0>}}
  }
  Spacer {widgetClassPath=&Spacer, parent=VBox_MainMenuButtons, widget=Spacer, properties {size=<1 16>}, slot {class=VerticalBoxSlot}}
  RespawnButton {
   widgetClassPath=&MainMenuButton
   bIsVariable=true
   parent=VBox_MainMenuButtons
   widget=RespawnButton
   properties {
    defaultText=Respawn
    mobileButtonHeight=80
    onButtonBaseClicked=(&self:WidgetTree.RespawnButton.HandleOnButtonClicked)
    onButtonBaseDoubleClicked=(&self:WidgetTree.RespawnButton.HandleOnButtonDoubleClicked)
    onButtonBaseHovered=(&self:WidgetTree.RespawnButton.HandleOnButtonHovered)
    onButtonBaseUnhovered=(&self:WidgetTree.RespawnButton.HandleOnButtonUnhovered)
    onSelectedChangedBase=(&self:WidgetTree.RespawnButton.HandleOnSelectedChanged)
    style=&Sidebar_ButtonStyle-Tab-Main-Recolor_C
   }
   slot {class=VerticalBoxSlot, properties {padding=<ltrb 0 8 0 0>}}
  }
  TeamSelectButton {
   widgetClassPath=&MainMenuButton
   bIsVariable=true
   parent=VBox_MainMenuButtons
   widget=TeamSelectButton
   properties {
    defaultText=Team Select
    mobileButtonHeight=80
    onButtonBaseClicked=(&self:WidgetTree.TeamSelectButton.HandleOnButtonClicked)
    onButtonBaseDoubleClicked=(&self:WidgetTree.TeamSelectButton.HandleOnButtonDoubleClicked)
    onButtonBaseHovered=(&self:WidgetTree.TeamSelectButton.HandleOnButtonHovered)
    onButtonBaseUnhovered=(&self:WidgetTree.TeamSelectButton.HandleOnButtonUnhovered)
    onSelectedChangedBase=(&self:WidgetTree.TeamSelectButton.HandleOnSelectedChanged)
    style=&Sidebar_ButtonStyle-Tab-Main-Recolor_C
    defaultBrush {
     bIsDynamicallyLoaded=false
     drawAs=Image
     imageType=NoImage
     mirroring=NoMirror
     resourceName=None
     resourceObject=/FortniteCoreUI/Athena/HUD/Art/AliveTeamsIco
     tiling=NoTile
     imageSize=<38 38>
     margin=<ltrb 0 0 0 0>
     outlineSettings {
      bUseBrushTransparency=false
      roundingType=HalfHeightRadius
      width=0
      color {colorUseRule=UseColor_Specified, specifiedColor=<rgba 0 0 0 0>}
      cornerRadii=<0 0 0 0>
     }
     tintColor {colorUseRule=UseColor_Specified, specifiedColor=<rgba 1 1 1 1>}
     uVRegion {bIsValid=false, max=<0 0>, min=<0 0>}
    }
   }
   slot {class=VerticalBoxSlot, properties {padding=<ltrb 0 8 0 0>}}
  }
  AthenaMainMenuClearPlayerDataButton {
   widgetClassPath=&MainMenuButton
   bIsVariable=true
   parent=VBox_MainMenuButtons
   widget=AthenaMainMenuClearPlayerDataButton
   properties {
    defaultText=Reset Progress
    mobileButtonHeight=80
    onButtonBaseClicked=(&self:WidgetTree.AthenaMainMenuClearPlayerDataButton.HandleOnButtonClicked)
    onButtonBaseDoubleClicked=(&self:WidgetTree.AthenaMainMenuClearPlayerDataButton.HandleOnButtonDoubleClicked)
    onButtonBaseHovered=(&self:WidgetTree.AthenaMainMenuClearPlayerDataButton.HandleOnButtonHovered)
    onButtonBaseUnhovered=(&self:WidgetTree.AthenaMainMenuClearPlayerDataButton.HandleOnButtonUnhovered)
    onSelectedChangedBase=(&self:WidgetTree.AthenaMainMenuClearPlayerDataButton.HandleOnSelectedChanged)
    style=&Sidebar_ButtonStyle-Tab-Main-Recolor_C
    defaultBrush {
     bIsDynamicallyLoaded=false
     drawAs=Image
     imageType=NoImage
     mirroring=NoMirror
     resourceName=None
     resourceObject=/Game/Athena/HUD/Art/ClearPersistenceIslandDataIcon
     tiling=NoTile
     imageSize=<38 38>
     margin=<ltrb 0 0 0 0>
     outlineSettings {
      bUseBrushTransparency=false
      roundingType=HalfHeightRadius
      width=0
      color {colorUseRule=UseColor_Specified, specifiedColor=<rgba 0 0 0 0>}
      cornerRadii=<0 0 0 0>
     }
     tintColor {colorUseRule=UseColor_Specified, specifiedColor=<rgba 1 1 1 1>}
     uVRegion {bIsValid=false, max=<0 0>, min=<0 0>}
    }
   }
   slot {class=VerticalBoxSlot, properties {padding=<ltrb 0 8 0 0>}}
  }
  Button_MyIsland {
   widgetClassPath=&MainMenuButton
   bIsVariable=true
   parent=VBox_MainMenuButtons
   widget=Button_MyIsland
   properties {
    defaultText=My Island
    mobileButtonHeight=80
    onButtonBaseClicked=(&self:WidgetTree.Button_MyIsland.HandleOnButtonClicked)
    onButtonBaseDoubleClicked=(&self:WidgetTree.Button_MyIsland.HandleOnButtonDoubleClicked)
    onButtonBaseHovered=(&self:WidgetTree.Button_MyIsland.HandleOnButtonHovered)
    onButtonBaseUnhovered=(&self:WidgetTree.Button_MyIsland.HandleOnButtonUnhovered)
    onSelectedChangedBase=(&self:WidgetTree.Button_MyIsland.HandleOnSelectedChanged)
    style=&Sidebar_ButtonStyle-Tab-Main-Recolor_C
   }
   slot {class=VerticalBoxSlot, properties {padding=<ltrb 0 8 0 0>}}
  }
  Button_BackToHub {
   widgetClassPath=&MainMenuButton
   bInherited=true
   bIsVariable=true
   parent=VBox_MainMenuButtons
   widget=Button_BackToHub
   properties {
    defaultText=Back to Hub
    mobileButtonHeight=80
    onButtonBaseClicked=(&self:WidgetTree.Button_BackToHub.HandleOnButtonClicked)
    onButtonBaseDoubleClicked=(&self:WidgetTree.Button_BackToHub.HandleOnButtonDoubleClicked)
    onButtonBaseHovered=(&self:WidgetTree.Button_BackToHub.HandleOnButtonHovered)
    onButtonBaseUnhovered=(&self:WidgetTree.Button_BackToHub.HandleOnButtonUnhovered)
    onSelectedChangedBase=(&self:WidgetTree.Button_BackToHub.HandleOnSelectedChanged)
    style=&Sidebar_ButtonStyle-Tab-Main-Recolor_C
   }
   slot {class=VerticalBoxSlot, properties {padding=<ltrb 0 8 0 0>}}
  }
  ExportIslandButton {
   widgetClassPath=&MainMenuButton
   bIsVariable=true
   parent=VBox_MainMenuButtons
   widget=ExportIslandButton
   properties {
    defaultText="Export Island (Developer)"
    mobileButtonHeight=80
    onButtonBaseClicked=(&self:WidgetTree.ExportIslandButton.HandleOnButtonClicked)
    onButtonBaseDoubleClicked=(&self:WidgetTree.ExportIslandButton.HandleOnButtonDoubleClicked)
    onButtonBaseHovered=(&self:WidgetTree.ExportIslandButton.HandleOnButtonHovered)
    onButtonBaseUnhovered=(&self:WidgetTree.ExportIslandButton.HandleOnButtonUnhovered)
    onSelectedChangedBase=(&self:WidgetTree.ExportIslandButton.HandleOnSelectedChanged)
    style=&Sidebar_ButtonStyle-Tab-Main-Recolor_C
   }
   slot {class=VerticalBoxSlot, properties {padding=<ltrb 0 8 0 0>}}
  }
  ModeratorMode {
   widgetClassPath=&Overlay
   bIsVariable=true
   parent=VBox_MainMenuButtons
   widget=ModeratorMode
   properties {visibility=Collapsed}
   slot {class=VerticalBoxSlot, properties {padding=<ltrb 0 8 0 0>}}
  }
  EnableModeratorModeButton {
   widgetClassPath=&MainMenuButton
   bIsVariable=true
   parent=ModeratorMode
   widget=EnableModeratorModeButton
   properties {
    defaultText=Enable Moderator Mode
    mobileButtonHeight=80
    onButtonBaseClicked=(&self:WidgetTree.EnableModeratorModeButton.HandleOnButtonClicked)
    onButtonBaseDoubleClicked=(&self:WidgetTree.EnableModeratorModeButton.HandleOnButtonDoubleClicked)
    onButtonBaseHovered=(&self:WidgetTree.EnableModeratorModeButton.HandleOnButtonHovered)
    onButtonBaseUnhovered=(&self:WidgetTree.EnableModeratorModeButton.HandleOnButtonUnhovered)
    onSelectedChangedBase=(&self:WidgetTree.EnableModeratorModeButton.HandleOnSelectedChanged)
    style=&Sidebar_ButtonStyle-Tab-Main-Recolor_C
   }
   slot {class=OverlaySlot, properties {horizontalAlignment=HAlign_Fill, verticalAlignment=VAlign_Fill}}
  }
  DisableModeratorModeButton {
   widgetClassPath=&MainMenuButton
   bIsVariable=true
   parent=ModeratorMode
   widget=DisableModeratorModeButton
   properties {
    defaultText=Disable Moderator Mode
    mobileButtonHeight=80
    onButtonBaseClicked=(&self:WidgetTree.DisableModeratorModeButton.HandleOnButtonClicked)
    onButtonBaseDoubleClicked=(&self:WidgetTree.DisableModeratorModeButton.HandleOnButtonDoubleClicked)
    onButtonBaseHovered=(&self:WidgetTree.DisableModeratorModeButton.HandleOnButtonHovered)
    onButtonBaseUnhovered=(&self:WidgetTree.DisableModeratorModeButton.HandleOnButtonUnhovered)
    onSelectedChangedBase=(&self:WidgetTree.DisableModeratorModeButton.HandleOnSelectedChanged)
    style=/Game/UI/Sidebar/Sidebar_ButtonStyle-Tab-Main-Recolor-Red.Sidebar_ButtonStyle-Tab-Main-Recolor-Red_C
   }
   slot {class=OverlaySlot, properties {horizontalAlignment=HAlign_Fill, verticalAlignment=VAlign_Fill}}
  }
  ModerationModeLoading {
   widgetClassPath=&HorizontalBox
   bIsVariable=true
   parent=ModeratorMode
   widget=ModerationModeLoading
   slot {class=OverlaySlot, properties {horizontalAlignment=HAlign_Fill, verticalAlignment=VAlign_Fill}}
  }
  SizeBox_5069 {
   widgetClassPath=&SizeBox
   parent=ModerationModeLoading
   widget=SizeBox_5069
   properties {bOverride_HeightOverride=true, heightOverride=48, widthOverride=96}
   slot {class=HorizontalBoxSlot}
  }
  ProgressSpinner {
   widgetClassPath=&Image
   bIsVariable=true
   parent=SizeBox_5069
   widget=ProgressSpinner
   properties {
    visibility=SelfHitTestInvisible
    brush {
     bIsDynamicallyLoaded=false
     drawAs=Image
     imageType=NoImage
     mirroring=NoMirror
     resourceName=None
     resourceObject=/FortUILibrary/UIKit/Throbber/MI_UIKit_Monochrome_Throbber
     tiling=NoTile
     imageSize=<64 64>
     margin=<ltrb 0 0 0 0>
     outlineSettings {
      bUseBrushTransparency=false
      roundingType=HalfHeightRadius
      width=0
      color {colorUseRule=UseColor_Specified, specifiedColor=<rgba 0 0 0 0>}
      cornerRadii=<0 0 0 0>
     }
     tintColor {colorUseRule=UseColor_Specified, specifiedColor=<rgba 1 1 1 1>}
     uVRegion {bIsValid=false, max=<0 0>, min=<0 0>}
    }
    colorAndOpacity=<rgba 0.2705 0.6724 0.9911 1>
   }
   slot {class=SizeBoxSlot, properties {horizontalAlignment=HAlign_Center, verticalAlignment=VAlign_Center, padding=<ltrb 4 4 4 4>}}
  }
  ModeratorModeLoadingMessage {
   widgetClassPath=&CommonTextBlock
   bIsVariable=true
   parent=ModerationModeLoading
   widget=ModeratorModeLoadingMessage
   properties {
    justification=Center
    textTransformPolicy=ToUpper
    colorAndOpacity {colorUseRule=UseColor_Specified, specifiedColor=<rgba 0.2159 0.6584 1 1>}
    font {
     bForceMonospaced=false
     bMaterialIsStencil=false
     fontMaterial=null
     fontObject=&BurbankSmall
     letterSpacing=0
     monospacedWidth=1
     size=15
     skewAmount=0
     typefaceFontName=Black
     outlineSettings {
      bApplyOutlineToDropShadows=false
      bMiteredCorners=false
      bSeparateFillAlpha=false
      outlineMaterial=null
      outlineSize=0
      outlineColor=<rgba 0 0 0 1>
     }
    }
    shadowOffset=<0 0>
   }
   slot {class=HorizontalBoxSlot, properties {horizontalAlignment=HAlign_Left, verticalAlignment=VAlign_Center}}
  }
  VisibilitySubGameButton {
   widgetClassPath=&CommonVisibilityWidgetLegacy
   parent=VBox_MainMenuButtons
   widget=VisibilitySubGameButton
   properties {bShowForAndroid=false, bShowForErebus=false, bShowForIOS=false, bShowForSwitch2=false, padding=<ltrb 0 8 0 0>}
   slot {class=VerticalBoxSlot}
  }
  Button_SubgameSelection {
   widgetClassPath=&MainMenuButton
   bInherited=true
   bIsVariable=true
   parent=VisibilitySubGameButton
   widget=Button_SubgameSelection
   properties {
    defaultText=Select Game Mode
    mobileButtonHeight=80
    onButtonBaseClicked=(&self:WidgetTree.Button_SubgameSelection.HandleOnButtonClicked)
    onButtonBaseDoubleClicked=(&self:WidgetTree.Button_SubgameSelection.HandleOnButtonDoubleClicked)
    onButtonBaseHovered=(&self:WidgetTree.Button_SubgameSelection.HandleOnButtonHovered)
    onButtonBaseUnhovered=(&self:WidgetTree.Button_SubgameSelection.HandleOnButtonUnhovered)
    onSelectedChangedBase=(&self:WidgetTree.Button_SubgameSelection.HandleOnSelectedChanged)
   }
   slot {class=BorderSlot, properties {padding=<ltrb 0 8 0 0>}}
  }
  VerticalBox_ExtensionContainer {
   widgetClassPath=&VerticalBox
   bInherited=true
   bIsVariable=true
   parent=VBox_MainMenuButtons
   widget=VerticalBox_ExtensionContainer
   slot {class=VerticalBoxSlot, properties {size {sizeRule=Fill, value=1}}}
  }
  EntryBox_RegistryButtons {
   widgetClassPath=/Script/FortniteUI.FortDynamicEntryBox
   bInherited=true
   bIsVariable=true
   parent=VBox_MainMenuButtons
   widget=EntryBox_RegistryButtons
   properties {
    entryBoxType=Vertical
    entryWidgetClass=/Game/UI/Frontend/MainMenu/MainMenuButton.MainMenuButton_C
    numDesignerPreviewEntries=4
    entrySpacing=<0 8>
   }
   slot {class=VerticalBoxSlot, properties {padding=<ltrb 0 8 0 0>}}
  }
  MinigameButtons {
   widgetClassPath=/FortniteCoreUI/InGame/Creative/MinigameButtons
   bInherited=true
   bIsVariable=true
   parent=VBox_MainMenuButtons
   widget=MinigameButtons
   slot {class=VerticalBoxSlot, properties {padding=<ltrb 0 8 0 0>}}
  }
  CheatsButton {
   widgetClassPath=&MainMenuButton
   bIsVariable=true
   parent=VBox_MainMenuButtons
   widget=CheatsButton
   properties {
    defaultText="Cheats (Developer)"
    mobileButtonHeight=80
    onButtonBaseClicked=(&self:WidgetTree.CheatsButton.HandleOnButtonClicked)
    onButtonBaseDoubleClicked=(&self:WidgetTree.CheatsButton.HandleOnButtonDoubleClicked)
    onButtonBaseHovered=(&self:WidgetTree.CheatsButton.HandleOnButtonHovered)
    onButtonBaseUnhovered=(&self:WidgetTree.CheatsButton.HandleOnButtonUnhovered)
    onSelectedChangedBase=(&self:WidgetTree.CheatsButton.HandleOnSelectedChanged)
    style=&Sidebar_ButtonStyle-Tab-Main-Recolor_C
   }
   slot {class=VerticalBoxSlot, properties {padding=<ltrb 0 8 0 0>}}
  }
  Button_Locker {
   widgetClassPath=&MainMenuButton
   bInherited=true
   bIsVariable=true
   parent=VBox_MainMenuButtons
   widget=Button_Locker
   properties {
    defaultText="Locker (Developer)"
    mobileButtonHeight=80
    onButtonBaseClicked=(&self:WidgetTree.Button_Locker.HandleOnButtonClicked)
    onButtonBaseDoubleClicked=(&self:WidgetTree.Button_Locker.HandleOnButtonDoubleClicked)
    onButtonBaseHovered=(&self:WidgetTree.Button_Locker.HandleOnButtonHovered)
    onButtonBaseUnhovered=(&self:WidgetTree.Button_Locker.HandleOnButtonUnhovered)
    onSelectedChangedBase=(&self:WidgetTree.Button_Locker.HandleOnSelectedChanged)
    style=&Sidebar_ButtonStyle-Tab-Main-Recolor_C
   }
   slot {class=VerticalBoxSlot, properties {padding=<ltrb 0 8 0 0>}}
  }
  ConsoleButton {
   widgetClassPath=&MainMenuButton
   bIsVariable=true
   parent=VBox_MainMenuButtons
   widget=ConsoleButton
   properties {
    defaultText="Console (Developer)"
    mobileButtonHeight=80
    onButtonBaseClicked=(&self:WidgetTree.ConsoleButton.HandleOnButtonClicked)
    onButtonBaseDoubleClicked=(&self:WidgetTree.ConsoleButton.HandleOnButtonDoubleClicked)
    onButtonBaseHovered=(&self:WidgetTree.ConsoleButton.HandleOnButtonHovered)
    onButtonBaseUnhovered=(&self:WidgetTree.ConsoleButton.HandleOnButtonUnhovered)
    onSelectedChangedBase=(&self:WidgetTree.ConsoleButton.HandleOnSelectedChanged)
    style=&Sidebar_ButtonStyle-Tab-Main-Recolor_C
    visibility=Collapsed
   }
   slot {class=VerticalBoxSlot, properties {padding=<ltrb 0 8 0 0>}}
  }
  ConsoleCommandPrompt {
   widgetClassPath=/FortUILibrary/UIKit/InputField/WBP_UIKit_InputField
   bIsVariable=true
   parent=VBox_MainMenuButtons
   widget=ConsoleCommandPrompt
   properties {
    hint Text=Console Command
    icon=null
    iconClass=null
    minHeight=48
    style=/FortUILibrary/Foundation/Buttons/ButtonStyles/Tab/ButtonStyle-Tab-Main-Recolor.ButtonStyle-Tab-Main-Recolor_C
    text="Console (DEVELOPER)"
    useBackground=false
    useIcon=false
    visibility=Collapsed
    iconProperties {
     material=null
     image Size=<50 50>
     placement {
      horizontal Alignment=HAlign_Center
      vertical Alignment=VAlign_Center
      grid {column=0, column Span=0, layer=0, row=1, row Span=0, nudge=<0 0>}
      padding=<ltrb 0 0 0 0>
     }
    }
   }
   slot {class=VerticalBoxSlot}
  }
  CommonVisibilityWidget_1 {
   widgetClassPath=&CommonVisibilityWidgetLegacy
   parent=VBox_MainMenuButtons
   widget=CommonVisibilityWidget_1
   properties {bShowForPS4=false, bShowForPS5=false, bShowForXBox=false, bShowForXSX=false, visibility=Collapsed, padding=<ltrb 0 0 0 0>}
   slot {class=VerticalBoxSlot}
  }
  Button_MobileBack {
   widgetClassPath=&MainMenuButton
   bInherited=true
   bIsVariable=true
   parent=VBox_MainMenuButtons
   widget=Button_MobileBack
   properties {
    defaultText=Back To Party Hub
    mobileButtonHeight=80
    onButtonBaseClicked=(&self:WidgetTree.Button_MobileBack.HandleOnButtonClicked)
    onButtonBaseDoubleClicked=(&self:WidgetTree.Button_MobileBack.HandleOnButtonDoubleClicked)
    onButtonBaseHovered=(&self:WidgetTree.Button_MobileBack.HandleOnButtonHovered)
    onButtonBaseUnhovered=(&self:WidgetTree.Button_MobileBack.HandleOnButtonUnhovered)
    onSelectedChangedBase=(&self:WidgetTree.Button_MobileBack.HandleOnSelectedChanged)
    style=&Sidebar_ButtonStyle-Tab-Main-Recolor-Red_C
   }
   slot {class=VerticalBoxSlot, properties {padding=<ltrb 0 8 0 0>}}
  }
  LiveStreamMenuButtonToggle {
   widgetClassPath=/FortniteCoreUI/Frontend/LiveStreamVIdeo/LiveStreamMenuButtonToggle
   bIsVariable=true
   parent=VBox_MainMenuButtons
   widget=LiveStreamMenuButtonToggle
   properties {
    onButtonBaseClicked=(&self:WidgetTree.LiveStreamMenuButtonToggle.HandleOnButtonClicked)
    onButtonBaseDoubleClicked=(&self:WidgetTree.LiveStreamMenuButtonToggle.HandleOnButtonDoubleClicked)
    onButtonBaseHovered=(&self:WidgetTree.LiveStreamMenuButtonToggle.HandleOnButtonHovered)
    onButtonBaseUnhovered=(&self:WidgetTree.LiveStreamMenuButtonToggle.HandleOnButtonUnhovered)
    onSelectedChangedBase=(&self:WidgetTree.LiveStreamMenuButtonToggle.HandleOnSelectedChanged)
   }
   slot {class=VerticalBoxSlot, properties {padding=<ltrb 0 8 0 0>}}
  }
  VBox_MainMenuButtons_Reporting {
   widgetClassPath=&VerticalBox
   bIsVariable=true
   parent=Switcher_Menus
   widget=VBox_MainMenuButtons_Reporting
   properties {flowDirectionPreference=Culture}
   slot {class=WidgetSwitcherSlot}
  }
  Button_ReportBugOrPlayer {
   widgetClassPath=&MainMenuButton
   bIsVariable=true
   parent=VBox_MainMenuButtons_Reporting
   widget=Button_ReportBugOrPlayer
   properties {
    defaultText="Report a bug, player, or island"
    mobileButtonHeight=80
    onButtonBaseClicked=(&self:WidgetTree.Button_ReportBugOrPlayer.HandleOnButtonClicked)
    onButtonBaseDoubleClicked=(&self:WidgetTree.Button_ReportBugOrPlayer.HandleOnButtonDoubleClicked)
    onButtonBaseHovered=(&self:WidgetTree.Button_ReportBugOrPlayer.HandleOnButtonHovered)
    onButtonBaseUnhovered=(&self:WidgetTree.Button_ReportBugOrPlayer.HandleOnButtonUnhovered)
    onSelectedChangedBase=(&self:WidgetTree.Button_ReportBugOrPlayer.HandleOnSelectedChanged)
    style=&Sidebar_ButtonStyle-Tab-Main-Recolor_C
   }
   slot {class=VerticalBoxSlot, properties {padding=<ltrb 0 8 0 0>}}
  }
  Button_CodeOfConduct {
   widgetClassPath=&MainMenuButton
   bInherited=true
   bIsVariable=true
   parent=VBox_MainMenuButtons_Reporting
   widget=Button_CodeOfConduct
   properties {
    defaultText=Community Rules
    mobileButtonHeight=80
    onButtonBaseClicked=(&self:WidgetTree.Button_CodeOfConduct.HandleOnButtonClicked)
    onButtonBaseDoubleClicked=(&self:WidgetTree.Button_CodeOfConduct.HandleOnButtonDoubleClicked)
    onButtonBaseHovered=(&self:WidgetTree.Button_CodeOfConduct.HandleOnButtonHovered)
    onButtonBaseUnhovered=(&self:WidgetTree.Button_CodeOfConduct.HandleOnButtonUnhovered)
    onSelectedChangedBase=(&self:WidgetTree.Button_CodeOfConduct.HandleOnSelectedChanged)
    showExternalLinkIcon=true
    style=&Sidebar_ButtonStyle-Tab-Main-Recolor_C
   }
   slot {class=VerticalBoxSlot, properties {padding=<ltrb 0 8 0 0>}}
  }
  VBox_MainMenuButtons_SupportAndRefunds {
   widgetClassPath=&VerticalBox
   bIsVariable=true
   parent=Switcher_Menus
   widget=VBox_MainMenuButtons_SupportAndRefunds
   properties {flowDirectionPreference=Culture}
   slot {class=WidgetSwitcherSlot}
  }
  Button_Support {
   widgetClassPath=&MainMenuButton
   bInherited=true
   bIsVariable=true
   parent=VBox_MainMenuButtons_SupportAndRefunds
   widget=Button_Support
   properties {
    defaultText=Support
    mobileButtonHeight=80
    onButtonBaseClicked=(&self:WidgetTree.Button_Support.HandleOnButtonClicked)
    onButtonBaseDoubleClicked=(&self:WidgetTree.Button_Support.HandleOnButtonDoubleClicked)
    onButtonBaseHovered=(&self:WidgetTree.Button_Support.HandleOnButtonHovered)
    onButtonBaseUnhovered=(&self:WidgetTree.Button_Support.HandleOnButtonUnhovered)
    onSelectedChangedBase=(&self:WidgetTree.Button_Support.HandleOnSelectedChanged)
    showExternalLinkIcon=true
    style=&Sidebar_ButtonStyle-Tab-Main-Recolor_C
   }
   slot {class=VerticalBoxSlot, properties {padding=<ltrb 0 8 0 0>}}
  }
  ReturnOrCancelPurchase {
   widgetClassPath=&MainMenuButton
   bIsVariable=true
   parent=VBox_MainMenuButtons_SupportAndRefunds
   widget=ReturnOrCancelPurchase
   properties {
    defaultText=Return or cancel purchase
    mobileButtonHeight=80
    onButtonBaseClicked=(&self:WidgetTree.ReturnOrCancelPurchase.HandleOnButtonClicked)
    onButtonBaseDoubleClicked=(&self:WidgetTree.ReturnOrCancelPurchase.HandleOnButtonDoubleClicked)
    onButtonBaseHovered=(&self:WidgetTree.ReturnOrCancelPurchase.HandleOnButtonHovered)
    onButtonBaseUnhovered=(&self:WidgetTree.ReturnOrCancelPurchase.HandleOnButtonUnhovered)
    onSelectedChangedBase=(&self:WidgetTree.ReturnOrCancelPurchase.HandleOnSelectedChanged)
    style=&Sidebar_ButtonStyle-Tab-Main-Recolor_C
   }
   slot {class=VerticalBoxSlot, properties {padding=<ltrb 0 8 0 0>}}
  }
  VBox_MainMenuButtons_Legal {
   widgetClassPath=&VerticalBox
   bIsVariable=true
   parent=Switcher_Menus
   widget=VBox_MainMenuButtons_Legal
   properties {flowDirectionPreference=Culture}
   slot {class=WidgetSwitcherSlot}
  }
  Button_LegalInformation {
   widgetClassPath=&MainMenuButton
   bIsVariable=true
   parent=VBox_MainMenuButtons_Legal
   widget=Button_LegalInformation
   properties {
    defaultText=Legal Information
    mobileButtonHeight=80
    onButtonBaseClicked=(&self:WidgetTree.Button_LegalInformation.HandleOnButtonClicked)
    onButtonBaseDoubleClicked=(&self:WidgetTree.Button_LegalInformation.HandleOnButtonDoubleClicked)
    onButtonBaseHovered=(&self:WidgetTree.Button_LegalInformation.HandleOnButtonHovered)
    onButtonBaseUnhovered=(&self:WidgetTree.Button_LegalInformation.HandleOnButtonUnhovered)
    onSelectedChangedBase=(&self:WidgetTree.Button_LegalInformation.HandleOnSelectedChanged)
    style=&Sidebar_ButtonStyle-Tab-Main-Recolor_C
   }
   slot {class=VerticalBoxSlot, properties {padding=<ltrb 0 8 0 0>}}
  }
  ButtonCredits {
   widgetClassPath=&MainMenuButton
   bIsVariable=true
   parent=VBox_MainMenuButtons_Legal
   widget=ButtonCredits
   properties {
    defaultText=Credits
    mobileButtonHeight=80
    onButtonBaseClicked=(&self:WidgetTree.ButtonCredits.HandleOnButtonClicked)
    onButtonBaseDoubleClicked=(&self:WidgetTree.ButtonCredits.HandleOnButtonDoubleClicked)
    onButtonBaseHovered=(&self:WidgetTree.ButtonCredits.HandleOnButtonHovered)
    onButtonBaseUnhovered=(&self:WidgetTree.ButtonCredits.HandleOnButtonUnhovered)
    onSelectedChangedBase=(&self:WidgetTree.ButtonCredits.HandleOnSelectedChanged)
    style=&Sidebar_ButtonStyle-Tab-Main-Recolor_C
   }
   slot {class=VerticalBoxSlot, properties {padding=<ltrb 0 8 0 0>}}
  }
  Text_BackToHubDisabledReason {
   widgetClassPath=&CommonTextBlock
   bInherited=true
   bIsVariable=true
   parent=VerticalBox_0
   widget=Text_BackToHubDisabledReason
   properties {
    autoWrapText=true
    bIsScrollingEnabled=false
    justification=Center
    text=Cannot access Back To Hub while using Looking for Party
    visibility=Collapsed
    colorAndOpacity {colorUseRule=UseColor_Specified, specifiedColor=<rgba 0.588 0.588 0.588 1>}
    font {
     bForceMonospaced=false
     bMaterialIsStencil=false
     fontMaterial=null
     fontObject=&BurbankSmall
     letterSpacing=0
     monospacedWidth=1
     size=18
     skewAmount=0
     typefaceFontName=Black
     outlineSettings {
      bApplyOutlineToDropShadows=false
      bMiteredCorners=false
      bSeparateFillAlpha=false
      outlineMaterial=null
      outlineSize=0
      outlineColor=<rgba 0 0 0 1>
     }
    }
    shadowOffset=<0 0>
   }
   slot {class=VerticalBoxSlot, properties {verticalAlignment=VAlign_Center, padding=<ltrb 30 10 30 30>}}
  }
  TextLowPowerMode {
   widgetClassPath=&CommonTextBlock
   bIsVariable=true
   parent=VerticalBox_0
   widget=TextLowPowerMode
   properties {
    justification=Center
    scrollStyle=/FortUILibrary/Styles/Typography/Foundation/TextScrollStyle-NoFade.TextScrollStyle-NoFade_C
    text=Warning: Low Performance Mode
    textTransformPolicy=ToUpper
    visibility=Collapsed
    colorAndOpacity {colorUseRule=UseColor_Specified, specifiedColor=<rgba 0.9301 1 0.0097 1>}
    font {
     bForceMonospaced=false
     bMaterialIsStencil=false
     fontMaterial=null
     fontObject=&BurbankSmall
     letterSpacing=0
     monospacedWidth=1
     size=18
     skewAmount=0
     typefaceFontName=Black
     outlineSettings {
      bApplyOutlineToDropShadows=false
      bMiteredCorners=false
      bSeparateFillAlpha=false
      outlineMaterial=null
      outlineSize=0
      outlineColor=<rgba 0 0 0 1>
     }
    }
    shadowOffset=<0 0>
   }
   slot {class=VerticalBoxSlot, properties {verticalAlignment=VAlign_Center, padding=<ltrb 0 10 0 30>}}
  }
  SafeZoneRatingIcons {
   widgetClassPath=&SafeZone
   bIsVariable=true
   parent=VerticalBox_0
   widget=SafeZoneRatingIcons
   properties {padBottom=false, padLeft=false, padTop=false}
   slot {class=VerticalBoxSlot}
  }
  RatingIcons_RatingIcons {
   widgetClassPath=/Game/UI/Frontend/Misc/RestrictedCountryRatingIcons
   bInherited=true
   bIsVariable=true
   parent=SafeZoneRatingIcons
   widget=RatingIcons_RatingIcons
   slot {class=SafeZoneSlot, properties {hAlign=HAlign_Center}}
  }
  SidebarBackButton {
   widgetClassPath=/FortUILibrary/Foundation/Buttons/BackButtonSmall
   bInherited=true
   bIsVariable=true
   parent=Overlay_0
   widget=SidebarBackButton
   properties {visibility=Collapsed}
   slot {class=OverlaySlot}
  }
  Button_LeaveExperience {widgetClassPath=&CommonButtonLegacy, bInherited=true, widget=null}
 ]
}
