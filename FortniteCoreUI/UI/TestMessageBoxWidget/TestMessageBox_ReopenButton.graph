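# Test UI widget that shows a scaled-up, horizontally centered "Reopen Message Box" button for bringing the test message box back.
# Clicking the button gets every TestMessageBoxWidget instance, including nested ones, and calls showMenu on each.
# It holds no reference to the message box and creates none, so it only affects instances that already exist.
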WidgetBlueprint /FortniteCoreUI/UI/TestMessageBoxWidget/TestMessageBox_ReopenButton
graphs [EventGraph {nodes [
   "On Clicked (TA_Generic01_Button)"!() -> (OutputDelegate, then, widget:User Widget Object Reference, event:Pointer Event Structure) {is=title K2Node_ComponentBoundEvent}
   > n1 = GetAllWidgetsOfClass!(
    execute
    WidgetClass:User Widget Class Reference=&/TestMessageBoxWidget.TestMessageBoxWidget_C
    TopLevelOnly:bool=false
   ) -> (then, FoundWidgets:Array of Test Message Box Widget Object References)
   > n2 = For Each Loop(Array:Array of Test Message Box Widget Object References=@n1) -> (Array Element:Test Message Box Widget Object Reference)
   > showMenu(self:Test Message Box Widget Object Reference=@n2)
   Event PreConstruct()
   Event Construct()
   Event Tick()
  ]}]
variables []
widgets {
 info {inheritedWidgetCount=0, parentClass=&UserWidget, rootWidgetClass=&Overlay, widgetCount=3}
 widgets [
  Overlay_0 {widgetClassPath=&Overlay, widget=Overlay_0}
  ScaleBox_0 {
   widgetClassPath=&ScaleBox
   parent=Overlay_0
   widget=ScaleBox_0
   properties {stretch=UserSpecified, userSpecifiedScale=2}
   slot {class=OverlaySlot, properties {horizontalAlignment=HAlign_Center, padding=<ltrb 0 5 0 0>}}
  }
  TA_Generic01_Button {
   widgetClassPath=&TA_Generic01_Button
   parent=ScaleBox_0
   widget=TA_Generic01_Button
   properties {label=Reopen Message Box, buttonColor_Default=<rgba 1 0.1355 0 1>}
   slot {class=ScaleBoxSlot}
  }
 ]
}
