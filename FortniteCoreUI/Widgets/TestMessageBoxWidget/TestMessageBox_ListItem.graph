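# List row widget for a test message box, built on the tech-art editor button base, with a text label, optional icon and optional button.
# On data update it copies the item's message struct, shows and sizes any valid icon, and collapses the button when its text is empty.
# Button clicks broadcast onButtonClicked with the row itself.
# Overrides expose widgets and a stored object, and tint the edge bar by selection.
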WidgetBlueprint /FortniteCoreUI/Widgets/TestMessageBoxWidget/TestMessageBox_ListItem
graphs [
 EventGraph {nodes [
   "On Clicked (TA_Generic01_Button)"!() -> (OutputDelegate, then, widget:User Widget Object Reference, event:Pointer Event Structure) {is=title K2Node_ComponentBoundEvent}
   > Call On Button Clicked!(execute:Exec, self:Test Message Box List Item Object Reference, buttonOwner:Test Message Box List Item Object Reference=Self-Reference()) -> (then:Exec) {is=title K2Node_CallDelegate}
   n1 = Event updateAfterDataChange!() -> (OutputDelegate, then, currentListItemData:Object Reference) {is=title K2Node_Event}
   > Parent: updateAfterDataChange!(execute, currentListItemData:Object Reference=@n1.currentListItemData) -> (then) {is=title K2Node_CallParentFunction}
   > n2 = Cast To TestMessageBox_ListItemData!(execute, Object=@n1.currentListItemData) -> (then, CastFailed, AsTest Message Box List Item Data:Test Message Box List Item Data Object Reference) {is=title K2Node_DynamicCast}
   > Set currentMessageStruct(currentMessageStruct:Test Message Struct Structure=Get messageStruct!(self:Test Message Box List Item Data Object Reference=@n2) -> (messageStruct:Test Message Struct Structure)) -> (Output_Get:Test Message Struct Structure)
   > n3 = Sequence()
   n4 = Get!() -> (
    currentMessageStruct_message_2_#1:String
    currentMessageStruct_buttonText_5_#2:String
    currentMessageStruct_Tab_6_#3:String
    currentMessageStruct_object_9_#4:Object Reference
    currentMessageStruct_icon_12_#5:Texture 2D Object Reference
   )
   > Is Valid(InputObject=@n4.currentMessageStruct_icon_12_#5)
   n5 = Get imgIcon() -> (imgIcon:Image Object Reference)
   > SetVisibility(self=@n5, InVisibility=SelfHitTestInvisible)
   > SetBrush(
    self=@n5
    InBrush=MakeBrushFromTexture!(Texture:Texture 2D Object Reference=@n4.currentMessageStruct_icon_12_#5, Width:int=24, Height:int=24) -> (ReturnValue:Slate Brush Structure)
   )
   n6 = SlotAsHorizontalBoxSlot!(Widget:Widget Object Reference=Get imgIcon()) -> (ReturnValue:Horizontal Box Slot Object Reference)
   > SetSize!(
    execute
    self:Horizontal Box Slot Object Reference=@n6
    InSize:Slate Child Size Structure=Make Slate Child Size!(
     Value:float=1
     SizeRule:ESlateSizeRule Enum=Get brushFill!(self:TA Generic List Item Data Object Reference=@n2) -> (brushFill:ESlateSizeRule Enum=Automatic)
    ) -> (SlateChildSize:Slate Child Size Structure) {is=title K2Node_MakeStruct}
   ) -> (then)
   > SetHorizontalAlignment(
    self:Horizontal Box Slot Object Reference=@n6
    InHorizontalAlignment=Get brushAlignment!(self:TA Generic List Item Data Object Reference=@n2) -> (brushAlignment:EHorizontalAlignment Enum=HAlign_Fill)
   )
   n7 = Get TA_Generic01_Button() -> (TA_Generic01_Button:TA Generic 01 Button Object Reference)
   n8 = Get()
   setLabel!(
    execute=@n3.then_1
    self:TA Base Editor Button Widget Object Reference=@n7
    label:String=@n8.currentMessageStruct_buttonText_5_#2
   ) -> (then)
   n9 = IsEmpty(InString=@n8.currentMessageStruct_buttonText_5_#2)
   > SetVisibility(self=@n7, InVisibility=Select(Option 0=Visible, Option 1=Collapsed, Index=@n9))
   > SetPadding!(
    execute
    self:Horizontal Box Slot Object Reference=SlotAsHorizontalBoxSlot(Widget=@n7)
    InPadding:Margin Structure=Select(Option 0:Margin Structure=Make Margin(Left=0, Top=0, Right=5, Bottom=0), Option 1:Margin Structure, Index=@n9) -> (ReturnValue:Margin Structure)
   ) -> (then)
   addTestMessage!(execute, message:Test Message Struct Structure, forceShowMenu:bool=false) -> (then)
  ]}
 getBackgroundWidget {nodes [
   getBackgroundWidget() {is=title K2Node_FunctionEntry}
   > Return Node!(execute, backgroundWidget:Widget Object Reference=Get borderMain() -> (borderMain:Border Object Reference))
  ]}
 getLabelWidget {nodes [
   getLabelWidget() {is=title K2Node_FunctionEntry}
   > Return Node!(execute, labelWidget:Widget Object Reference=Get txtMain() -> (txtMain:Text Object Reference))
  ]}
 getStoredObject {nodes [
   getStoredObject() {is=title K2Node_FunctionEntry}
   > n1 = Cast To TestMessageBox_ListItemData(Object=Get currentListItemData() -> (currentListItemData:Object Reference))
   n2 = Get objects!(self:TA Generic List Item Data Object Reference=@n1) -> (objects:Array of Object References)
   > n3 = Branch(Condition=Is Valid Index!(TargetArray:Array of Object References=@n2, IndexToTest:int) -> (ReturnValue:bool))
   > Return Node!(execute, object:Object Reference="Get (a copy)"(Array:Array of Object References=@n2) -> (Output:Object Reference))
   Return Node(execute=@n3.else)
  ]}
 onButtonClicked {nodes [onButtonClicked!() -> (then, buttonOwner:Test Message Box List Item Object Reference) {is=title K2Node_FunctionEntry}]}
 setTextAlignment {nodes [
   n1 = setTextAlignment!() -> (then, alignment:EHorizontalAlignment Enum=HAlign_Fill) {is=title K2Node_FunctionEntry}
   > n2 = Cast To ScaleBoxSlot!(execute, Object=Get Slot!(self:Widget Object Reference=Get txtMain()) -> (Slot:Panel Slot Object Reference)) -> (then, CastFailed, AsScale Box Slot:Scale Box Slot Object Reference) {is=title K2Node_DynamicCast}
   > SetHorizontalAlignment(self:Scale Box Slot Object Reference=@n2, InHorizontalAlignment=@n1)
  ]}
 validateBackgroundColor {nodes [
   validateBackgroundColor() {is=title K2Node_FunctionEntry}
   > SetColorAndOpacity(
    self:Image Object Reference=Get imageEdge() -> (imageEdge:Image Object Reference)
    InColorAndOpacity:Linear Color Structure=SelectColor(A=(R=0.0284,G=0.3763,B=1,A=1), B=(R=0.8123,G=0.8217,B=0.8385,A=1), bPickA=getSelected!() -> (isSelected:bool))
   )
   > n1 = Parent: validateBackgroundColor!(execute) -> (then, handled:bool) {is=title K2Node_CallParentFunction}
   > Return Node!(execute, handled:bool=@n1)
  ]}
]
variables [
 currentMessageStruct {type=TestMessage_Struct, default {buttonText="", icon=null, message="", object=null, tab=""}}
 onButtonClicked {}
]
widgets {
 info {
  inheritedWidgetCount=0
  parentClass=/Game/Blueprints/Blutilities/ForTechArt/Widgets/Buttons/TA_BaseEditorButtonWidget.TA_BaseEditorButtonWidget_C
  rootWidgetClass=&Overlay
  widgetCount=10
 }
 widgets [
  Overlay_7 {widgetClassPath=&Overlay, widget=Overlay_7}
  VerticalBox_0 {
   widgetClassPath=&VerticalBox
   parent=Overlay_7
   widget=VerticalBox_0
   slot {class=OverlaySlot, properties {horizontalAlignment=HAlign_Fill}}
  }
  borderMain {
   widgetClassPath=&Border
   bIsVariable=true
   parent=VerticalBox_0
   widget=borderMain
   properties {padding=<ltrb 3 3 3 3>}
   slot {class=VerticalBoxSlot}
  }
  HorizontalBox_0 {
   widgetClassPath=&HorizontalBox
   parent=borderMain
   widget=HorizontalBox_0
   slot {class=BorderSlot, properties {padding=<ltrb 3 3 3 3>}}
  }
  ScaleBox_1 {
   widgetClassPath=&ScaleBox
   parent=HorizontalBox_0
   widget=ScaleBox_1
   properties {stretch=UserSpecified, userSpecifiedScale=0.8}
   slot {class=HorizontalBoxSlot, properties {verticalAlignment=VAlign_Center}}
  }
  TA_Generic01_Button {
   widgetClassPath=&TA_Generic01_Button
   parent=ScaleBox_1
   widget=TA_Generic01_Button
   properties {label=Button, visibility=Collapsed}
   slot {class=ScaleBoxSlot}
  }
  imgIcon {
   widgetClassPath=&Image
   bIsVariable=true
   parent=HorizontalBox_0
   widget=imgIcon
   properties {
    visibility=Collapsed
    brush {
     bIsDynamicallyLoaded=false
     drawAs=Image
     imageType=NoImage
     mirroring=NoMirror
     resourceName=None
     resourceObject=null
     tiling=NoTile
     imageSize=<20 20>
     margin=<ltrb 0 0 0 0>
     outlineSettings {
      bUseBrushTransparency=false
      roundingType=HalfHeightRadius
      width=0
      color {colorUseRule=UseColor_Specified, specifiedColor=<rgba 0 0 0 0>}
      cornerRadii=<0 0 0 0>
     }
     tintColor {colorUseRule=UseColor_Specified, specifiedColor=<rgba 1 1 1 1>}
     uVRegion {bIsValid=false, max=<0 0>, min=<0 0>}
    }
   }
   slot {class=HorizontalBoxSlot, properties {horizontalAlignment=HAlign_Center, verticalAlignment=VAlign_Center}}
  }
  ScaleBox_0 {
   widgetClassPath=&ScaleBox
   parent=HorizontalBox_0
   widget=ScaleBox_0
   properties {clipping=ClipToBounds, stretch=None, stretchDirection=DownOnly}
   slot {class=HorizontalBoxSlot, properties {padding=<ltrb 5 0 0 0>, size {sizeRule=Fill, value=1}}}
  }
  txtMain {
   widgetClassPath=&TextBlock
   bIsVariable=true
   parent=ScaleBox_0
   widget=txtMain
   properties {
    autoWrapText=true
    clipping=ClipToBounds
    colorAndOpacity {colorUseRule=UseColor_Specified, specifiedColor=<rgba 0.0104 0.0104 0.0104 1>}
    font {
     bForceMonospaced=false
     bMaterialIsStencil=false
     fontMaterial=null
     fontObject=&Roboto
     letterSpacing=0
     monospacedWidth=1
     size=12
     skewAmount=0
     typefaceFontName=Bold
     outlineSettings {
      bApplyOutlineToDropShadows=false
      bMiteredCorners=false
      bSeparateFillAlpha=false
      outlineMaterial=null
      outlineSize=0
      outlineColor=<rgba 0 0 0 1>
     }
    }
    shadowColorAndOpacity=<rgba 0 0 0 0.2>
   }
   slot {class=ScaleBoxSlot, properties {horizontalAlignment=HAlign_Fill}}
  }
  imageEdge {
   widgetClassPath=&Image
   bIsVariable=true
   parent=VerticalBox_0
   widget=imageEdge
   properties {
    brush {
     bIsDynamicallyLoaded=false
     drawAs=Image
     imageType=NoImage
     mirroring=NoMirror
     resourceName=None
     resourceObject=null
     tiling=NoTile
     imageSize=<32 2>
     margin=<ltrb 0 0 0 0>
     outlineSettings {
      bUseBrushTransparency=false
      roundingType=HalfHeightRadius
      width=0
      color {colorUseRule=UseColor_Specified, specifiedColor=<rgba 0 0 0 0>}
      cornerRadii=<0 0 0 0>
     }
     tintColor {colorUseRule=UseColor_Specified, specifiedColor=<rgba 1 1 1 1>}
     uVRegion {bIsValid=false, max=<0 0>, min=<0 0>}
    }
    colorAndOpacity=<rgba 0.2031 0.2031 0.2031 1>
   }
   slot {class=VerticalBoxSlot}
  }
 ]
}
