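# Tab button widget for the TestMessageBoxWidget UI, built on the tech-art TA_BaseEditorButtonWidget, with optional icon, label, close button and bottom selection bar.
# New list data's first bool hides the close button when true.
# validateVisuals applies the icon brush and transform, collapses a missing icon or empty label, and tints the bar by selection.
# Clicking close broadcasts onCloseClicked with itself and the item's displayString.
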WidgetBlueprint /FortniteCoreUI/Widgets/TestMessageBoxWidget/TestMessage_TabButton
graphs [
 EventGraph {nodes [
   "On Clicked (btnClose)"!() -> (OutputDelegate, then, widget:User Widget Object Reference, event:Pointer Event Structure) {is=title K2Node_ComponentBoundEvent}
   > n1 = Cast To TA_GenericListItemData!(execute, Object=Get currentListItemData() -> (currentListItemData:Object Reference)) -> (then, CastFailed, AsTA Generic List Item Data:TA Generic List Item Data Object Reference) {is=title K2Node_DynamicCast}
   > Call On Close Clicked!(
    execute:Exec
    self:Test Message Tab Button Object Reference
    widget:User Widget Object Reference=Self-Reference()
    tab:String=Get displayString!(self:TA Generic List Item Data Object Reference=@n1) -> (displayString:String)
   ) -> (then:Exec) {is=title K2Node_CallDelegate}
  ]}
 getBackgroundWidget {nodes [
   getBackgroundWidget() {is=title K2Node_FunctionEntry}
   > Return Node!(execute, backgroundWidget:Widget Object Reference=Get borderMain() -> (borderMain:Border Object Reference))
  ]}
 getLabelWidget {nodes [
   getLabelWidget() {is=title K2Node_FunctionEntry}
   > Return Node!(execute, labelWidget:Widget Object Reference=Get txtMain() -> (txtMain:Text Object Reference))
  ]}
 handleNewListDataSet {nodes [
   n1 = handleNewListDataSet!() -> (then, currentListItemData:Object Reference) {is=title K2Node_FunctionEntry}
   > Parent: handleNewListDataSet!(execute, currentListItemData:Object Reference=@n1) -> (then) {is=title K2Node_CallParentFunction}
   > n2 = Cast To TA_GenericListItemData(Object=Get currentListItemData())
   n3 = Get bools!(self:TA Generic List Item Data Object Reference=@n2) -> (bools:Array of Booleans)
   > SetVisibility(
    self=Get btnClose() -> (btnClose:TA Generic 01 Button Object Reference)
    InVisibility=Select(Option 0=Visible, Option 1=Collapsed, Index="Get (a copy)"(Array:Array of Booleans=@n3) -> (Output:bool))
   )
  ]}
 onCloseClicked {nodes [onCloseClicked!() -> (then, widget:User Widget Object Reference, tab:String) {is=title K2Node_FunctionEntry}]}
 validateVisuals {nodes [
   validateVisuals() {is=title K2Node_FunctionEntry}
   > Parent: validateVisuals!(execute) -> (then, " ":bool) {is=title K2Node_CallParentFunction}
   > n1 = Sequence!(execute) -> (then_0, then_1, then_2:Exec)
   > n2 = Is Valid(InputObject=Break Slate Brush!(SlateBrush:"Slate Brush Structure (by ref)"=Get iconBrush() -> (iconBrush:Slate Brush Structure)) -> (ResourceObject:Object Reference, OutlineSettings:Slate Brush Outline Settings Structure) {is=title K2Node_BreakStruct})
   SetVisibility(execute=@n2.Is Not Valid, self=Get imageIcon() -> (imageIcon:Image Object Reference), InVisibility=Collapsed)
   SetBrush(execute=@n2.Is Valid, self=Get imageIcon(), InBrush=Get iconBrush())
   > SetVisibility(self=Get imageIcon(), InVisibility=HitTestInvisible)
   > SetRenderTransform!(
    execute
    self:Widget Object Reference=Get imageIcon()
    InTransform:Widget Transform Structure=Get iconTransform() -> (iconTransform:Widget Transform Structure)
   ) -> (then)
   n3 = Get label() -> (label:String)
   SetVisibility(
    execute=@n1.then_1
    self=Get txtMain()
    InVisibility=Select(Option 0=Collapsed, Option 1=HitTestInvisible, Index="Not Equal Exactly (String)"(A=@n3))
   )
   SetColorAndOpacity(
    execute=@n1.then_2
    self:Image Object Reference=Get Image_59() -> (Image_59:Image Object Reference)
    InColorAndOpacity:Linear Color Structure=SelectColor(A=(R=1,G=0.9188,B=0.5345,A=1), B=(R=0,G=0,B=0,A=0.5), bPickA=getSelected!() -> (isSelected:bool))
   )
   Return Node!(execute, " ":bool=false)
  ]}
]
variables [
 iconBrush {
  type=SlateBrush
  default {
   bIsDynamicallyLoaded=false
   drawAs=Image
   imageType=NoImage
   mirroring=NoMirror
   resourceName=None
   resourceObject=null
   tiling=NoTile
   imageSize=<32 32>
   margin=<ltrb 0 0 0 0>
   outlineSettings {
    bUseBrushTransparency=false
    roundingType=HalfHeightRadius
    width=0
    color {colorUseRule=UseColor_Specified, specifiedColor=<rgba 0 0 0 0>}
    cornerRadii=<0 0 0 1>
   }
   tintColor {colorUseRule=UseColor_Specified, specifiedColor=<rgba 1 1 1 1>}
   uVRegion {bIsValid=false, max=<0 0>, min=<0 0>}
  }
 }
 iconTransform {type=WidgetTransform, default {angle=0, scale=<1 1>, shear=<0 0>, translation=<0 0>}}
 onCloseClicked {}
]
widgets {
 info {
  inheritedWidgetCount=0
  parentClass=/Game/Blueprints/Blutilities/ForTechArt/Widgets/Buttons/TA_BaseEditorButtonWidget.TA_BaseEditorButtonWidget_C
  rootWidgetClass=&Overlay
  widgetCount=9
 }
 widgets [
  Overlay_7 {widgetClassPath=&Overlay, widget=Overlay_7}
  Overlay_0 {
   widgetClassPath=&Overlay
   parent=Overlay_7
   widget=Overlay_0
   slot {class=OverlaySlot, properties {horizontalAlignment=HAlign_Fill}}
  }
  borderMain {
   widgetClassPath=&Border
   bIsVariable=true
   parent=Overlay_0
   widget=borderMain
   properties {padding=<ltrb 4 3 0 3>}
   slot {class=OverlaySlot, properties {horizontalAlignment=HAlign_Fill}}
  }
  HorizontalBox_0 {
   widgetClassPath=&HorizontalBox
   parent=borderMain
   widget=HorizontalBox_0
   slot {class=BorderSlot, properties {padding=<ltrb 4 3 0 3>}}
  }
  imageIcon {
   widgetClassPath=&Image
   bIsVariable=true
   parent=HorizontalBox_0
   widget=imageIcon
   slot {class=HorizontalBoxSlot, properties {horizontalAlignment=HAlign_Center, verticalAlignment=VAlign_Center}}
  }
  ScaleBox_0 {
   widgetClassPath=&ScaleBox
   parent=HorizontalBox_0
   widget=ScaleBox_0
   properties {clipping=ClipToBounds, stretchDirection=DownOnly}
   slot {class=HorizontalBoxSlot, properties {padding=<ltrb 0 0 3 0>, size {sizeRule=Fill, value=1}}}
  }
  txtMain {
   widgetClassPath=&TextBlock
   bIsVariable=true
   parent=ScaleBox_0
   widget=txtMain
   properties {
    clipping=ClipToBounds
    justification=Center
    colorAndOpacity {colorUseRule=UseColor_Specified, specifiedColor=<rgba 0.0104 0.0104 0.0104 1>}
    font {
     bForceMonospaced=false
     bMaterialIsStencil=false
     fontMaterial=null
     fontObject=&Roboto
     letterSpacing=0
     monospacedWidth=1
     size=10
     skewAmount=0
     typefaceFontName=Bold
     outlineSettings {
      bApplyOutlineToDropShadows=false
      bMiteredCorners=false
      bSeparateFillAlpha=false
      outlineMaterial=null
      outlineSize=0
      outlineColor=<rgba 0 0 0 1>
     }
    }
    shadowColorAndOpacity=<rgba 0 0 0 0.5>
   }
   slot {class=ScaleBoxSlot}
  }
  btnClose {
   widgetClassPath=&TA_Generic01_Button
   parent=HorizontalBox_0
   widget=btnClose
   properties {
    buttonColor_Default=<rgba 1 0 0.0184 1>
    iconBrush {
     bIsDynamicallyLoaded=false
     drawAs=Image
     imageType=NoImage
     mirroring=NoMirror
     resourceName=None
     resourceObject=/Game/UI/Icons/Icon_Seach_Close
     tiling=NoTile
     imageSize=<9 9>
     margin=<ltrb 0 0 0 0>
     outlineSettings {
      bUseBrushTransparency=false
      roundingType=HalfHeightRadius
      width=0
      color {colorUseRule=UseColor_Specified, specifiedColor=<rgba 0 0 0 0>}
      cornerRadii=<0 0 0 1>
     }
     tintColor {colorUseRule=UseColor_Specified, specifiedColor=<rgba 1 1 1 1>}
     uVRegion {bIsValid=false, max=<0 0>, min=<0 0>}
    }
   }
   slot {class=HorizontalBoxSlot, properties {horizontalAlignment=HAlign_Right}}
  }
  Image_59 {
   widgetClassPath=&Image
   bIsVariable=true
   parent=Overlay_0
   widget=Image_59
   properties {
    brush {
     bIsDynamicallyLoaded=false
     drawAs=Image
     imageType=NoImage
     mirroring=NoMirror
     resourceName=None
     resourceObject=null
     tiling=NoTile
     imageSize=<32 4>
     margin=<ltrb 0 0 0 0>
     outlineSettings {
      bUseBrushTransparency=false
      roundingType=HalfHeightRadius
      width=0
      color {colorUseRule=UseColor_Specified, specifiedColor=<rgba 0 0 0 0>}
      cornerRadii=<0 0 0 0>
     }
     tintColor {colorUseRule=UseColor_Specified, specifiedColor=<rgba 1 1 1 1>}
     uVRegion {bIsValid=false, max=<0 0>, min=<0 0>}
    }
    colorAndOpacity=<rgba 0 0 0 0.5>
   }
   slot {class=OverlaySlot, properties {horizontalAlignment=HAlign_Fill, verticalAlignment=VAlign_Bottom}}
  }
 ]
}
