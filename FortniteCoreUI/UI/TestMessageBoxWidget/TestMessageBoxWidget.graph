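# TestMessageBoxWidget is a draggable, resizable floating UMG message box that shows test messages in closable tabs.
# addMsg stores messages per tab name in a map, displaying those for the selected tab and rebroadcasting button clicks through a dispatcher.
# Closed tabs are only collapsed, so they can be restored.
# Tick forces the mouse cursor on; hiding injects a reopen button beside SidebarMainMenu if none exists and flashes an Escape-menu hint.
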WidgetBlueprint /FortniteCoreUI/UI/TestMessageBoxWidget/TestMessageBoxWidget
animations [showOpenMessage {
  binding_count=1
  length_seconds=3
  object=&self:showOpenMessage
  sequence {bindings [CommonTextBlock_44 {
     id=#1
     possessed_class=CommonTextBlock
     tracks [{
       type=MovieSceneFloatTrack
       display_name=Render Opacity
       sections [{
         type=MovieSceneFloatSection
         end_frame=60
         start_frame=0
         channels [None {type=MovieSceneScriptingFloatChannel, keys [[0,0],[5,1],[20,1],[60,0]]}]
        }]
      }]
    }]}
 }]
graphs [
 Append Message {nodes [
   n1 = Append Message!() -> (then, message:Test Message Struct Structure) {is=title K2Node_FunctionEntry}
   > n2 = Set l_message(l_message:Test Message Struct Structure=@n1) -> (Output_Get:Test Message Struct Structure)
   > n3 = isTabCurrent!(
    execute
    self
    tab:String=Break Test Message Struct!(TestMessage_Struct:"Test Message Struct Structure (by ref)"=@n2) -> (Tab_6_#2:String) {is=title K2Node_BreakStruct}
   ) -> (then, isCurrent:bool)
   > Branch(Condition=@n3)
   > createMessage!(execute, self, messageStruct:Test Message Struct Structure=Get l_message() -> (l_message:Test Message Struct Structure)) -> (then)
  ]}
 EventGraph {nodes [
   "On Clicked (btnClose)"!() -> (OutputDelegate, then, widget:User Widget Object Reference, event:Pointer Event Structure) {is=title K2Node_ComponentBoundEvent}
   > hideMenu()
   "On Clicked (btnClearMessages)"!() -> (OutputDelegate, then, widget:User Widget Object Reference, event:Pointer Event Structure) {is=title K2Node_ComponentBoundEvent}
   > clearCurrentLog()
   "On Clicked (btnRestoreTabs)"!() -> (OutputDelegate, then, widget:User Widget Object Reference, event:Pointer Event Structure) {is=title K2Node_ComponentBoundEvent}
   > restoreClosedTabs()
   "On Item Selection Changed (listTabs)"!() -> (
    OutputDelegate
    then
    selectedData:Array of Object References
    selectedWidgets:Array of User Widget Object References
    selectedIndices:Array of Integers
   ) {is=title K2Node_ComponentBoundEvent}
   > n1 = getSelectedData!(execute, self:TA Generic 01 List Object Reference=Get listTabs() -> (listTabs:TA Generic 01 List Object Reference)) -> (then, selectedData:Array of Object References)
   > Branch(Condition=Is Valid Index!(TargetArray:Array of Object References=@n1, IndexToTest:int) -> (ReturnValue:bool))
   > n2 = Cast To TA_GenericListItemData!(execute, Object="Get (a copy)"(Array:Array of Object References=@n1) -> (Output:Object Reference)) -> (then, CastFailed, AsTA Generic List Item Data:TA Generic List Item Data Object Reference) {is=title K2Node_DynamicCast}
   > PopulateTab!(execute, self, tab:String=Get displayString!(self:TA Generic List Item Data Object Reference=@n2) -> (displayString:String)) -> (then)
   Event showWidget() {is=title K2Node_Event}
   > showMenu()
   Event Construct()
   > showMenu()
   > initDefaultTab()
   Event Tick()
   > forceShowMouseCursor()
   Event hideWidget() {is=title K2Node_Event}
   > hideMenu()
   n3 = Event addMsg!() -> (OutputDelegate, then, msg:Test Message Struct Structure) {is=title K2Node_Event}
   > addMessage!(execute, self, message:Test Message Struct Structure=@n3.msg) -> (then)
  ]}
 On_borderResize_MouseButtonDown {nodes [
   n1 = On_borderResize_MouseButtonDown!() -> (then, MyGeometry:Geometry Structure, MouseEvent:"Pointer Event Structure (by ref)") {is=title K2Node_FunctionEntry}
   > n2 = Branch(Condition="Equal (Key)"!(
     A:Key Structure=Get Effecting Button!(Input:"Pointer Event Structure (by ref)"=@n1.MouseEvent) -> (ReturnValue:Key Structure) {is=title K2Node_CallFunction}
     B:Key Structure=LeftMouseButton
    ) -> (ReturnValue:bool))
   n3 = SlotAsCanvasSlot!(Widget:Widget Object Reference=Get borderMain() -> (borderMain:Border Object Reference)) -> (ReturnValue:Canvas Panel Slot Object Reference)
   n4 = Get Last Screen Space Position!(Input:"Pointer Event Structure (by ref)"=@n1.MouseEvent) -> (ReturnValue:Vector 2D Structure) {is=title K2Node_CallFunction}
   n5 = GetCachedGeometry!(self:Widget Object Reference) -> (ReturnValue:"Geometry Structure (by ref)")
   > Set dragStartLocation(dragStartLocation:Vector 2D Structure=vector2d - vector2d!(
     A:Vector 2D Structure=GetSize!(self:Canvas Panel Slot Object Reference=@n3) -> (ReturnValue:Vector 2D Structure)
     B:Vector 2D Structure="Absolute to Local (Coordinates)"!(Geometry:"Geometry Structure (by ref)"=@n5, AbsoluteCoordinate:Vector 2D Structure=@n4) -> (ReturnValue:Vector 2D Structure)
    ) -> (ReturnValue:Vector 2D Structure)) -> (Output_Get:Vector 2D Structure)
   n6 = LockMouse!(Reply:"Event Reply Structure (by ref)"=Handled()) -> (ReturnValue:Event Reply Structure)
   n7 = Get borderResize() -> (borderResize:Border Object Reference)
   > Return Node!(
    execute
    ReturnValue:Event Reply Structure=CaptureMouse!(
     Reply:"Event Reply Structure (by ref)"=SetUserFocus!(Reply:"Event Reply Structure (by ref)"=@n6, FocusWidget:Widget Object Reference=@n7, bInAllUsers:bool=true) -> (ReturnValue:Event Reply Structure)
     CapturingWidget:Widget Object Reference=@n7
    ) -> (ReturnValue:Event Reply Structure)
   )
   Return Node(execute=@n2.else, ReturnValue=Unhandled())
  ]}
 On_borderResize_MouseButtonUp {nodes [
   n1 = On_borderResize_MouseButtonUp!() -> (then, MyGeometry:Geometry Structure, MouseEvent:"Pointer Event Structure (by ref)") {is=title K2Node_FunctionEntry}
   > Branch(Condition="Equal (Key)"(A=Get Effecting Button(Input=@n1.MouseEvent), B=LeftMouseButton))
   > Set dragStartLocation(dragStartLocation=(X=-1,Y=-1))
   > Return Node(ReturnValue=ReleaseMouseCapture!(Reply:"Event Reply Structure (by ref)"=Handled()) -> (ReturnValue:Event Reply Structure))
  ]}
 On_borderResize_MouseMove {nodes [
   n1 = On_borderResize_MouseMove!() -> (then, MyGeometry:Geometry Structure, MouseEvent:"Pointer Event Structure (by ref)") {is=title K2Node_FunctionEntry}
   > n2 = Branch(Condition=HasMouseCapture!(self:Widget Object Reference=Get borderResize()) -> (ReturnValue:bool))
   n3 = Get Last Screen Space Position(Input=@n1.MouseEvent)
   n4 = GetCachedGeometry()
   n5 = BreakVector2D(InVec=vector2d + vector2d(
     A:Vector 2D Structure="Absolute to Local (Coordinates)"(Geometry=@n4, AbsoluteCoordinate=@n3)
     B:Vector 2D Structure=Get dragStartLocation() -> (dragStartLocation:Vector 2D Structure)
    ) -> (ReturnValue:Vector 2D Structure) {is=title K2Node_CommutativeAssociativeBinaryOperator})
   > SetSize!(
    execute
    self:Canvas Panel Slot Object Reference=SlotAsCanvasSlot(Widget=Get borderMain())
    InSize:Vector 2D Structure=MakeVector2D(X="Max (Float)"(A=@n5, B=200), Y="Max (Float)"(A=@n5.Y, B=200))
   ) -> (then)
   > Return Node(ReturnValue=Handled())
   Return Node(execute=@n2.else, ReturnValue=Unhandled())
  ]}
 On_borderTopBar_MouseButtonDown {nodes [
   n1 = On_borderTopBar_MouseButtonDown!() -> (then, MyGeometry:Geometry Structure, MouseEvent:"Pointer Event Structure (by ref)") {is=title K2Node_FunctionEntry}
   > n2 = Branch(Condition="Equal (Key)"(A=Get Effecting Button(Input=@n1.MouseEvent), B=LeftMouseButton))
   n3 = Get borderMain()
   > Set dragStartLocation(dragStartLocation="Absolute to Local (Coordinates)"(Geometry=GetCachedGeometry(self=@n3), AbsoluteCoordinate=Get Last Screen Space Position(Input=@n1.MouseEvent)))
   n4 = LockMouse(Reply=Handled())
   n5 = Get borderTopBar() -> (borderTopBar:Border Object Reference)
   > Return Node(ReturnValue=CaptureMouse(Reply=SetUserFocus(Reply=@n4, FocusWidget=@n5, bInAllUsers=true), CapturingWidget=@n5))
   Return Node(execute=@n2.else, ReturnValue=Unhandled())
  ]}
 On_borderTopBar_MouseButtonUp {nodes [
   n1 = On_borderTopBar_MouseButtonUp!() -> (then, MyGeometry:Geometry Structure, MouseEvent:"Pointer Event Structure (by ref)") {is=title K2Node_FunctionEntry}
   > Branch(Condition="Equal (Key)"(A=Get Effecting Button(Input=@n1.MouseEvent), B=LeftMouseButton))
   > Set dragStartLocation(dragStartLocation=(X=-1,Y=-1))
   > Return Node(ReturnValue=ReleaseMouseCapture(Reply=Handled()))
  ]}
 On_borderTopBar_MouseMove {nodes [
   n1 = On_borderTopBar_MouseMove!() -> (then, MyGeometry:Geometry Structure, MouseEvent:"Pointer Event Structure (by ref)") {is=title K2Node_FunctionEntry}
   > n2 = Branch(Condition=HasMouseCapture(self=Get borderTopBar()))
   n3 = Get Last Screen Space Position(Input=@n1.MouseEvent)
   n4 = GetCachedGeometry(self=Get CanvasPanel_0() -> (CanvasPanel_0:Canvas Panel Object Reference))
   n5 = BreakVector2D(InVec=vector2d - vector2d(A="Absolute to Local (Coordinates)"(Geometry=@n4, AbsoluteCoordinate=@n3), B=Get dragStartLocation()))
   n6 = BreakVector2D(InVec=vector2d / float!(A:Vector 2D Structure=GetViewportSize!() -> (ReturnValue:Vector 2D Structure), B:double=GetViewportScale!() -> (ReturnValue:float)) -> (ReturnValue:Vector 2D Structure))
   n7 = Get borderMain()
   n8 = BreakVector2D(InVec=GetLocalSize!(Geometry:"Geometry Structure (by ref)"=GetCachedGeometry(self=@n7)) -> (ReturnValue:Vector 2D Structure))
   n9 = "Min (Float)"(A:double=@n5, B:double=float - float(A=@n6, B=@n8) {is=title K2Node_CallFunction}) -> (ReturnValue:double) {is=title K2Node_CommutativeAssociativeBinaryOperator}
   n10 = "Min (Float)"(A=@n5.Y, B=float - float(A=@n6.Y, B=@n8.Y))
   > SetPosition!(
    execute
    self:Canvas Panel Slot Object Reference=SlotAsCanvasSlot(Widget=Get borderMain())
    InPosition:Vector 2D Structure=MakeVector2D(X="Max (Float)"(A=@n9, B=0), Y="Max (Float)"(A=@n10, B=0))
   ) -> (then)
   > Return Node(ReturnValue=Handled())
   Return Node(execute=@n2.else, ReturnValue=Unhandled())
  ]}
 PopulateTab {nodes [
   n1 = PopulateTab!() -> (then, tab:String) {is=title K2Node_FunctionEntry}
   > Set l_tab(l_tab:String=@n1) -> (Output_Get:String)
   n2 = Find(
    TargetMap:Map of Strings to Test Message Tabs Struct Structures=Get tabsAndMessages() -> (tabsAndMessages:Map of Strings to Test Message Tabs Struct Structures)
    Key=Get l_tab() -> (l_tab:String)
   ) -> (Value:Test Message Tabs Struct Structure)
   > Branch(Condition=@n2.ReturnValue)
   > Set l_messages(l_messages:Array of Test Message Struct Structures=Break Test Message Tabs Struct!(TestMessageTabs_Struct:"Test Message Tabs Struct Structure (by ref)"=@n2) -> (tab_16_#3:String, messages_17_#4:Array of Test Message Struct Structures) {is=title K2Node_BreakStruct}.messages_17_#4) -> (Output_Get:Array of Test Message Struct Structures)
   > clearItems(self:TA Generic 01 List Object Reference=Get listMessages() -> (listMessages:TA Generic 01 List Object Reference))
   > n3 = For Each Loop(Array:Array of Test Message Struct Structures=Get l_messages() -> (l_messages:Array of Test Message Struct Structures)) -> (Array Element:Test Message Struct Structure)
   > createMessage(messageStruct=@n3)
  ]}
 addMessage {nodes [
   n1 = addMessage!() -> (then, message:Test Message Struct Structure) {is=title K2Node_FunctionEntry}
   > Set l_newStruct(l_newStruct:Test Message Struct Structure=@n1) -> (Output_Get:Test Message Struct Structure)
   n2 = Break Test Message Struct!(TestMessage_Struct=@n1) -> (
    message_2_#3:String
    buttonText_5_#5:String
    Tab_6_#2
    object_9_#6:Object Reference
    icon_12_#4:Texture 2D Object Reference
   )
   > Set l_tab(l_tab=SelectString(A=Default, B=@n2.Tab_6_#2, bPickA=IsEmpty(InString=@n2.Tab_6_#2)))
   > Set members in Test Message Struct!(
    execute:Exec
    StructRef:"Test Message Struct Structure (by ref)"=Get l_newStruct() -> (l_newStruct:Test Message Struct Structure)
    Tab_6_#2:String=Get l_tab()
   ) -> (then:Exec, StructOut:"Test Message Struct Structure (by ref)") {is=title K2Node_SetFieldsInStruct}
   > n3 = Sequence()
   n4 = Find(TargetMap=Get tabsAndMessages(), Key=Get l_tab())
   > n5 = Branch(Condition=@n4.ReturnValue)
   > n6 = Set l_newMessages(l_newMessages:Array of Test Message Struct Structures=Break Test Message Tabs Struct(TestMessageTabs_Struct=@n4).messages_17_#4) -> (Output_Get:Array of Test Message Struct Structures)
   > Add(TargetArray:Array of Test Message Struct Structures=@n6, NewItem:"Test Message Struct Structure (by ref)"=Get l_newStruct()) {is=title K2Node_CallArrayFunction}
   > Add!(
    execute
    TargetMap:Map of Strings to Test Message Tabs Struct Structures=Get tabsAndMessages()
    Key:"String (by ref)"=Get l_tab()
    Value:"Test Message Tabs Struct Structure (by ref)"=Make Test Message Tabs Struct!(
     tab_16_#3:String=Get l_tab()
     messages_17_#4:Array of Test Message Struct Structures=Get l_newMessages() -> (l_newMessages:Array of Test Message Struct Structures)
    ) -> (TestMessageTabs_Struct:Test Message Tabs Struct Structure) {is=title K2Node_MakeStruct}
   ) -> (then) {is=title K2Node_CallFunction}
   n7 = Get l_newStruct()
   Add(
    execute=@n5.else
    TargetMap=Get tabsAndMessages()
    Key=Get l_tab()
    Value=Make Test Message Tabs Struct(
     tab_16_#3=Get l_tab()
     messages_17_#4=Make Array("[0]":Test Message Struct Structure=@n7) -> (Array:Array of Test Message Struct Structures)
    )
   )
   > n8 = addItemGeneric!(execute, self:TA Generic 01 List Object Reference=Get listTabs(), displayString:String=Get l_tab()) -> (then, widget:User Widget Object Reference, data:Object Reference)
   > n9 = Cast To TestMessage_TabButton!(execute, Object=@n8) -> (then, CastFailed, AsTest Message Tab Button:Test Message Tab Button Object Reference) {is=title K2Node_DynamicCast}
   > Bind Event to On Close Clicked(self:Test Message Tab Button Object Reference=@n9, Delegate=Create Event()) {is=title K2Node_AddDelegate}
   Append Message!(execute=@n3.then_1, self, message:Test Message Struct Structure=Get l_newStruct()) -> (then) {is=title K2Node_CallFunction}
  ]}
 clearCurrentLog {nodes [
   clearCurrentLog() {is=title K2Node_FunctionEntry}
   > n1 = Sequence!(execute) -> (then_0, then_1, then_2:Exec)
   > clearItems(self=Get listMessages())
   n2 = getSelectedItems!(execute=@n1.then_1, self:TA Generic 01 List Object Reference=Get listTabs()) -> (then, isAnySelected:bool, items:Array of User Widget Object References, selectedData:Array of Object References)
   > Branch(Condition=@n2)
   > n3 = Cast To TA_GenericListItemData(Object="Get (a copy)"(Array=@n2.selectedData))
   > Set l_currentTab(l_currentTab:String=Get displayString(self=@n3)) -> (Output_Get:String)
   > Add(
    TargetMap=Get tabsAndMessages()
    Key=Get l_currentTab() -> (l_currentTab:String)
    Value=Make Test Message Tabs Struct(tab_16_#3=Get l_currentTab())
   )
  ]}
 closeTab {nodes [
   n1 = closeTab!() -> (then, tab:String) {is=title K2Node_FunctionEntry}
   > Set l_tab(l_tab=@n1)
   > n2 = Sequence()
   > isTabCurrent(tab=Get l_tab())
   > setIndexSelected!(execute, self:TA Generic 01 List Object Reference=Get listTabs(), index:int=0, selected:bool=true, throwEvent:bool=true) -> (then)
   Add!(execute=@n2.then_1, TargetSet:Set of Strings=Get closedTabs() -> (closedTabs:Set of Strings), NewItem:"String (by ref)"=Get l_tab()) -> (then)
   > n3 = getListItems!(execute, self:TA Generic 01 List Object Reference=Get listTabs()) -> (then, items:Array of User Widget Object References)
   > n4 = For Each Loop with Break!(Exec:Exec, Array:Array of User Widget Object References=@n3, Break:Exec=@n6.then) -> (LoopBody:Exec, Array Element:User Widget Object Reference, Array Index:int, Completed:Exec)
   > n5 = Cast To TestMessage_TabButton(Object=@n4)
   > Branch(Condition="Equal Exactly (String)"(A=Get label!(self:TA Base Editor Button Widget Object Reference=@n5) -> (label:String), B=Get l_tab()))
   > n6 = SetVisibility(self=@n5, InVisibility=Collapsed)
   SetVisibility(execute=@n2.then_2, self=Get btnRestoreTabs() -> (btnRestoreTabs:TA Generic 01 Button Object Reference), InVisibility=Visible)
  ]}
 createMessage {nodes [
   n1 = createMessage!() -> (then, messageStruct:Test Message Struct Structure) {is=title K2Node_FunctionEntry}
   > n2 = Construct Object from Class(Class=/Game/Widgets/TestMessageBoxWidget/TestMessageBox_ListItemData.TestMessageBox_ListItemData_C, self=Self-Reference()) -> (ReturnValue:Test Message Box List Item Data Object Reference)
   > n3 = Set messageStruct!(execute, messageStruct:Test Message Struct Structure=@n1, self:Test Message Box List Item Data Object Reference=@n2) -> (then, Output_Get:Test Message Struct Structure)
   > Set displayString!(execute, displayString:String=Break Test Message Struct(TestMessage_Struct=@n3), self:TA Generic List Item Data Object Reference=@n2) -> (then, Output_Get:String)
   > n4 = addItem!(execute, self:TA Generic 01 List Object Reference=Get listMessages(), dataObject:Object Reference=@n2) -> (then, newItem:User Widget Object Reference, newData:Object Reference)
   > n5 = Cast To TestMessageBox_ListItem!(execute, Object=@n4) -> (then, CastFailed, AsTest Message Box List Item:Test Message Box List Item Object Reference) {is=title K2Node_DynamicCast}
   > Bind Event to On Button Clicked(self:Test Message Box List Item Object Reference=@n5, Delegate=Create Event()) {is=title K2Node_AddDelegate}
  ]}
 forceShowMouseCursor {nodes [
   forceShowMouseCursor() {is=title K2Node_FunctionEntry}
   > Branch(Condition=IsVisible())
   > n1 = Branch(Condition=Get bShowMouseCursor!(self:Player Controller Object Reference=GetPlayerController!(PlayerIndex:int=0) -> (ReturnValue:Player Controller Object Reference)) -> (bShowMouseCursor:bool))
   Set bShowMouseCursor!(execute=@n1.else, bShowMouseCursor:bool=true, self:Player Controller Object Reference=GetPlayerController(PlayerIndex=0)) -> (then, Output_Get:bool)
   > Set Input Mode Game And UI!(
    execute
    PlayerController:Player Controller Object Reference=GetPlayerController(PlayerIndex=0)
    InWidgetToFocus:Widget Object Reference=Self-Reference()
    InMouseLockMode:EMouseLockMode Enum=DoNotLock
    bHideCursorDuringCapture:bool=true
    bFlushInput:bool=false
   ) -> (then) {is=title K2Node_CallFunction}
   GetPlayerController(PlayerIndex=0)
  ]}
 handleMessageButtonClicked {nodes [
   n1 = handleMessageButtonClicked!() -> (then, buttonOwner:Test Message Box List Item Object Reference) {is=title K2Node_FunctionEntry}
   > Call On Message Button Clicked!(
    execute:Exec
    self:Test Message Box Widget Object Reference
    passedInObject:Test Message Struct Structure=Get currentMessageStruct!(self:Test Message Box List Item Object Reference=@n1) -> (currentMessageStruct:Test Message Struct Structure)
   ) -> (then:Exec) {is=title K2Node_CallDelegate}
  ]}
 handleTabCloseButton {nodes [
   n1 = handleTabCloseButton!() -> (then, widget:User Widget Object Reference, tab:String) {is=title K2Node_FunctionEntry}
   > closeTab!(execute, self, tab:String=@n1.tab) -> (then) {is=title K2Node_CallFunction}
  ]}
 hideMenu {nodes [
   hideMenu() {is=title K2Node_FunctionEntry}
   > n1 = Sequence!(execute) -> (then_0, then_1, then_2, then_3:Exec)
   > SetVisibility(self=Get borderMain(), InVisibility=Collapsed)
   n2 = GetAllWidgetsOfClass!(
    execute=@n1.then_1
    WidgetClass:User Widget Class Reference=&/TestMessageBox_ReopenButton.TestMessageBox_ReopenButton_C
    TopLevelOnly:bool=false
   ) -> (then, FoundWidgets:Array of Test Message Box Reopen Button Object References)
   > n3 = Branch(Condition=Is Valid Index(TargetArray:Array of Test Message Box Reopen Button Object References=@n2))
   > SetVisibility(
    self="Get (a copy)"(Array:Array of Test Message Box Reopen Button Object References=@n2) -> (Output:Test Message Box Reopen Button Object Reference)
    InVisibility=Visible
   )
   n4 = GetAllWidgetsOfClass(execute=@n3.else, WidgetClass=/FortniteCoreUI/UI/Sidebar/SidebarMainMenu.SidebarMainMenu_C, TopLevelOnly=false) -> (FoundWidgets:Array of Sidebar Main Menu Object References)
   n5 = Is Valid Index(TargetArray:Array of Sidebar Main Menu Object References=@n4)
   > FortShippingLog!(
    execute
    WorldContextObject:Object Reference
    InString:String=Append(A="Did we find Blade Menu? ", B="To String (Boolean)"(InBool=@n5))
    bLogAsWarning:bool=false
   ) -> (then)
   > Branch(Condition=@n5)
   n6 = GetParent!(self:Widget Object Reference="Get (a copy)"(Array:Array of Sidebar Main Menu Object References=@n4) -> (Output:Sidebar Main Menu Object Reference)) -> (ReturnValue:Panel Widget Object Reference)
   > FortShippingLog(InString=Append(A="Parent of Blade Menu: ", B=GetDisplayName(Object=@n6)), bLogAsWarning=false)
   > n7 = Cast To Overlay!(execute, Object=@n6) -> (then, CastFailed, AsOverlay:Overlay Object Reference) {is=title K2Node_DynamicCast}
   > FortShippingLog(InString=Append(A="Parent is an Overlay: ", B=GetDisplayName(Object=@n7)), bLogAsWarning=false)
   > n8 = Create Widget!(
    execute:Exec
    Class:User Widget Class Reference=&TestMessageBox_ReopenButton_C
    OwningPlayer:Player Controller Object Reference
   ) -> (then:Exec, ReturnValue:Test Message Box Reopen Button Object Reference)
   > AddChild(self=@n7, Content=@n8)
   n9 = SlotAsOverlaySlot(Widget=@n8)
   > SetHorizontalAlignment(self=@n9, InHorizontalAlignment=HAlign_Fill)
   > SetVerticalAlignment(self=@n9, InVerticalAlignment=VAlign_Fill)
   n10 = Cast To CanvasPanel!(execute=@n7.CastFailed, Object=@n6) -> (then, CastFailed, AsCanvas Panel:Canvas Panel Object Reference) {is=title K2Node_DynamicCast}
   > FortShippingLog(InString=Append(A="Parent is a canvas: ", B=GetDisplayName(Object=@n10)), bLogAsWarning=false)
   > n11 = Create Widget(Class=&TestMessageBox_ReopenButton_C)
   > AddChild(self=@n10, Content=@n11)
   n12 = SlotAsCanvasSlot(Widget=@n11)
   > SetAnchors!(
    execute
    self:Canvas Panel Slot Object Reference=@n12
    InAnchors:Anchors Structure=Make Anchors!(Minimum:Vector 2D Structure=(X=0,Y=0), Maximum:Vector 2D Structure=(X=1,Y=1)) -> (Anchors:Anchors Structure) {is=title K2Node_MakeStruct}
   ) -> (then)
   > SetOffsets!(execute, self:Canvas Panel Slot Object Reference=@n12, InOffset:Margin Structure=Make Margin(Left=0, Top=0, Right=0, Bottom=0)) -> (then)
   n13 = Branch(
    execute=@n1.then_2
    Condition="Equal (Enum)"(
     A:EMouseCaptureMode Enum=GetViewportMouseCaptureMode!() -> (ReturnValue:EMouseCaptureMode Enum=NoCapture)
     B:EMouseCaptureMode Enum=NoCapture
    )
   )
   n14 = GetPlayerController(PlayerIndex=0)
   Set bShowMouseCursor(execute=@n13.else, bShowMouseCursor=false, self=@n14)
   > SetInputMode_GameOnly!(execute, PlayerController:Player Controller Object Reference=@n14, bFlushInput:bool=false) -> (then)
   PlayAnimation(execute=@n1.then_3, InAnimation=Get showOpenMessage() -> (showOpenMessage:Widget Animation Object Reference), bRestoreState=true)
  ]}
 initDefaultTab {nodes [
   initDefaultTab() {is=title K2Node_FunctionEntry}
   > Add!(
    execute
    TargetMap:Map of Strings to Test Message Tabs Struct Structures=Get tabsAndMessages()
    Key:"String (by ref)"=Default
    Value:"Test Message Tabs Struct Structure (by ref)"
   ) -> (then)
   > n1 = Construct Object from Class(Class=/Game/Blueprints/Blutilities/ForTechArt/Data/TA_GenericListItemData.TA_GenericListItemData_C, self=Self-Reference()) -> (ReturnValue:TA Generic List Item Data Object Reference)
   > Set displayString(displayString=Default, self=@n1)
   > Set bools!(
    execute
    bools:Array of Booleans=Make Array("[0]":bool=true) -> (Array:Array of Booleans)
    self:TA Generic List Item Data Object Reference=@n1
   ) -> (then, Output_Get:Array of Booleans)
   n2 = Get listTabs()
   > addItem(self=@n2, dataObject=@n1)
   > setIndexSelected(self=@n2, index=0, selected=true, throwEvent=false)
  ]}
 isTabCurrent {nodes [
   n1 = isTabCurrent!() -> (then, tab:String) {is=title K2Node_FunctionEntry}
   > n2 = getSelectedItems(self=Get listTabs())
   > Branch(Condition=@n2)
   > n3 = Cast To TA_GenericListItemData(Object="Get (a copy)"(Array:Array of Object References=@n2.selectedData) -> (Output:Object Reference))
   > Branch(Condition="Equal Exactly (String)"(A=Get displayString(self=@n3), B=@n1))
   > Return Node!(execute, isCurrent:bool=true)
  ]}
 onMessageButtonClicked {nodes [onMessageButtonClicked!() -> (then, passedInObject:Test Message Struct Structure) {is=title K2Node_FunctionEntry}]}
 restoreClosedTabs {nodes [
   restoreClosedTabs() {is=title K2Node_FunctionEntry}
   > Clear!(execute, TargetSet:Set of Strings=Get closedTabs()) -> (then) {is=title K2Node_CallFunction}
   > SetVisibility(self=Get btnRestoreTabs(), InVisibility=Hidden)
   > n1 = getListItems(self=Get listTabs())
   > n2 = For Each Loop with Break(Array=@n1)
   > n3 = Cast To TestMessage_TabButton(Object=@n2)
   > SetVisibility(self=@n3, InVisibility=Visible)
  ]}
 showMenu {nodes [
   showMenu() {is=title K2Node_FunctionEntry}
   > SetVisibility(self=Get borderMain(), InVisibility=Visible)
   > Set bShowMouseCursor(bShowMouseCursor=true, self=GetPlayerController(PlayerIndex=0))
   n1 = GetPlayerController(PlayerIndex=0)
   > Set Input Mode Game And UI(PlayerController=@n1, InWidgetToFocus=Self-Reference(), InMouseLockMode=DoNotLock, bHideCursorDuringCapture=true, bFlushInput=false)
   > StopMovement(self:Controller Object Reference=@n1)
  ]}
]
variables [
 closedTabs {type="array<string>", default []}
 dragStartLocation {type=Vector2D, default=<0 0>}
 onMessageButtonClicked {}
 tabsAndMessages {type="map<TestMessageTabs_Struct>", default {}}
]
widgets {
 info {inheritedWidgetCount=0, parentClass=&UserWidget, rootWidgetClass=&CanvasPanel, widgetCount=18}
 widgets [
  CanvasPanel_0 {widgetClassPath=&CanvasPanel, bIsVariable=true, widget=CanvasPanel_0}
  borderMain {
   widgetClassPath=&Border
   bIsVariable=true
   parent=CanvasPanel_0
   widget=borderMain
   properties {brushColor=<rgba 0.0156 0.0156 0.0156 1>, padding=<ltrb 0 0 0 0>}
   slot {
    class=CanvasPanelSlot
    properties {layoutData {alignment=<0 0>, anchors {maximum=<0 0>, minimum=<0 0>}, offsets=<ltrb 0 0 727.6276 463.9339>}}
   }
  }
  VerticalBox_0 {
   widgetClassPath=&VerticalBox
   parent=borderMain
   widget=VerticalBox_0
   slot {class=BorderSlot, properties {padding=<ltrb 0 0 0 0>}}
  }
  borderTopBar {
   widgetClassPath=&Border
   bIsVariable=true
   parent=VerticalBox_0
   widget=borderTopBar
   properties {verticalAlignment=VAlign_Top, brushColor=<rgba 0.0433 0.2121 0.349 1>, padding=<ltrb 0 0 0 0>}
   slot {class=VerticalBoxSlot}
  }
  HorizontalBox_0 {
   widgetClassPath=&HorizontalBox
   parent=borderTopBar
   widget=HorizontalBox_0
   slot {class=BorderSlot, properties {verticalAlignment=VAlign_Top, padding=<ltrb 0 0 0 0>}}
  }
  CommonTextBlock_261 {
   widgetClassPath=&CommonTextBlock
   parent=HorizontalBox_0
   widget=CommonTextBlock_261
   properties {
    clipping=ClipToBounds
    text=Message Box
    visibility=HitTestInvisible
    font {
     bForceMonospaced=false
     bMaterialIsStencil=false
     fontMaterial=null
     fontObject=&NotoSans
     letterSpacing=0
     monospacedWidth=1
     size=16
     skewAmount=0
     typefaceFontName=NotoSans-Bold
     outlineSettings {
      bApplyOutlineToDropShadows=false
      bMiteredCorners=false
      bSeparateFillAlpha=false
      outlineMaterial=null
      outlineSize=0
      outlineColor=<rgba 0 0 0 1>
     }
    }
    shadowOffset=<0 0>
   }
   slot {class=HorizontalBoxSlot, properties {verticalAlignment=VAlign_Center, padding=<ltrb 10 5 10 5>, size {sizeRule=Fill, value=1}}}
  }
  btnClose {
   widgetClassPath=&TA_Generic01_Button
   parent=HorizontalBox_0
   widget=btnClose
   properties {
    buttonColor_Default=<rgba 1 0 0.0184 1>
    iconBrush {
     bIsDynamicallyLoaded=false
     drawAs=Image
     imageType=NoImage
     mirroring=NoMirror
     resourceName=None
     resourceObject=/Game/UI/Icons/Icon_Seach_Close
     tiling=NoTile
     imageSize=<32 32>
     margin=<ltrb 0 0 0 0>
     outlineSettings {
      bUseBrushTransparency=false
      roundingType=HalfHeightRadius
      width=0
      color {colorUseRule=UseColor_Specified, specifiedColor=<rgba 0 0 0 0>}
      cornerRadii=<0 0 0 1>
     }
     tintColor {colorUseRule=UseColor_Specified, specifiedColor=<rgba 1 1 1 1>}
     uVRegion {bIsValid=false, max=<0 0>, min=<0 0>}
    }
   }
   slot {class=HorizontalBoxSlot, properties {horizontalAlignment=HAlign_Right}}
  }
  Border_1 {
   widgetClassPath=&Border
   parent=VerticalBox_0
   widget=Border_1
   properties {
    background {
     bIsDynamicallyLoaded=false
     drawAs=Image
     imageType=NoImage
     mirroring=NoMirror
     resourceName=None
     resourceObject=null
     tiling=NoTile
     imageSize=<32 32>
     margin=<ltrb 0 0 0 0>
     outlineSettings {
      bUseBrushTransparency=false
      roundingType=HalfHeightRadius
      width=0
      color {colorUseRule=UseColor_Specified, specifiedColor=<rgba 0 0 0 0>}
      cornerRadii=<0 0 0 0>
     }
     tintColor {colorUseRule=UseColor_Specified, specifiedColor=<rgba 0.4531 0.4531 0.4531 1>}
     uVRegion {bIsValid=false, max=<0 0>, min=<0 0>}
    }
    brushColor=<rgba 0.2865 0.2865 0.2865 1>
    padding=<ltrb 4 10 4 0>
   }
   slot {class=VerticalBoxSlot}
  }
  listTabs {
   widgetClassPath=/Game/Blueprints/Blutilities/ForTechArt/Widgets/List/TA_Generic01_List
   bIsVariable=true
   parent=Border_1
   widget=listTabs
   properties {
    allowMultiSelect=false
    listItemClass=/FortniteCoreUI/Widgets/TestMessageBoxWidget/TestMessage_TabButton.TestMessage_TabButton_C
    scrollOrientation=Orient_Horizontal
   }
   slot {class=BorderSlot, properties {padding=<ltrb 4 10 4 0>}}
  }
  listMessages {
   widgetClassPath=&TA_Generic01_List
   bIsVariable=true
   parent=VerticalBox_0
   widget=listMessages
   properties {listItemClass=/FortniteCoreUI/Widgets/TestMessageBoxWidget/TestMessageBox_ListItem.TestMessageBox_ListItem_C}
   slot {class=VerticalBoxSlot, properties {size {sizeRule=Fill, value=1}}}
  }
  Border_0 {
   widgetClassPath=&Border
   parent=VerticalBox_0
   widget=Border_0
   properties {brushColor=<rgba 0 0 0 0.4>, padding=<ltrb 0 0 0 0>}
   slot {class=VerticalBoxSlot, properties {verticalAlignment=VAlign_Bottom}}
  }
  HorizontalBox_2 {
   widgetClassPath=&HorizontalBox
   parent=Border_0
   widget=HorizontalBox_2
   slot {class=BorderSlot, properties {padding=<ltrb 0 0 0 0>}}
  }
  btnClearMessages {
   widgetClassPath=&TA_Generic01_Button
   parent=HorizontalBox_2
   widget=btnClearMessages
   properties {label=Clear Messages, buttonColor_Default=<rgba 1 0.0994 0.0396 1>}
   slot {class=HorizontalBoxSlot, properties {horizontalAlignment=HAlign_Left, padding=<ltrb 5 5 5 5>, size {sizeRule=Fill, value=1}}}
  }
  btnRestoreTabs {
   widgetClassPath=&TA_Generic01_Button
   parent=HorizontalBox_2
   widget=btnRestoreTabs
   properties {label=Restore Closed Tabs, visibility=Hidden, buttonColor_Default=<rgba 0.153 0.4688 0.0448 1>}
   slot {class=HorizontalBoxSlot, properties {horizontalAlignment=HAlign_Left, padding=<ltrb 5 5 5 5>, size {sizeRule=Fill, value=1}}}
  }
  SizeBox_0 {
   widgetClassPath=&SizeBox
   parent=HorizontalBox_2
   widget=SizeBox_0
   properties {bOverride_HeightOverride=true, bOverride_WidthOverride=true, heightOverride=35, widthOverride=35}
   slot {class=HorizontalBoxSlot}
  }
  borderResize {
   widgetClassPath=&Border
   bIsVariable=true
   parent=SizeBox_0
   widget=borderResize
   properties {
    horizontalAlignment=HAlign_Right
    background {
     bIsDynamicallyLoaded=false
     drawAs=Image
     imageType=NoImage
     mirroring=NoMirror
     resourceName=None
     resourceObject=/Game/Blueprints/Blutilities/ForTechArt/ArtAssets/Textures/Checkbox/Icon_DragEdge
     tiling=NoTile
     imageSize=<40 40>
     margin=<ltrb 0 0 0 0>
     outlineSettings {
      bUseBrushTransparency=false
      roundingType=HalfHeightRadius
      width=0
      color {colorUseRule=UseColor_Specified, specifiedColor=<rgba 0 0 0 0>}
      cornerRadii=<0 0 0 0>
     }
     tintColor {colorUseRule=UseColor_Specified, specifiedColor=<rgba 1 1 1 1>}
     uVRegion {bIsValid=false, max=<0 0>, min=<0 0>}
    }
   }
   slot {class=SizeBoxSlot}
  }
  Overlay_103 {
   widgetClassPath=&Overlay
   parent=CanvasPanel_0
   widget=Overlay_103
   slot {class=CanvasPanelSlot, properties {layoutData {alignment=<0 0>, anchors {maximum=<1 1>, minimum=<0 0>}, offsets=<ltrb 0 0 0 0>}}}
  }
  CommonTextBlock_44 {
   widgetClassPath=&CommonTextBlock
   parent=Overlay_103
   widget=CommonTextBlock_44
   properties {
    renderOpacity=0
    text=Reopen Message Box in the Escape Menu
    visibility=HitTestInvisible
    colorAndOpacity {colorUseRule=UseColor_Specified, specifiedColor=<rgba 1 0.1682 0.0786 1>}
    font {
     bForceMonospaced=false
     bMaterialIsStencil=false
     fontMaterial=null
     fontObject=&NotoSans
     letterSpacing=0
     monospacedWidth=1
     size=48
     skewAmount=0
     typefaceFontName=NotoSans-Regular
     outlineSettings {
      bApplyOutlineToDropShadows=false
      bMiteredCorners=false
      bSeparateFillAlpha=false
      outlineMaterial=null
      outlineSize=5
      outlineColor=<rgba 0 0 0 1>
     }
    }
    shadowOffset=<0 0>
   }
   slot {class=OverlaySlot, properties {horizontalAlignment=HAlign_Center, verticalAlignment=VAlign_Center}}
  }
 ]
}
